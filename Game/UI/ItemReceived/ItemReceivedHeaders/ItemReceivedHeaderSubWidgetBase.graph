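# Base widget Blueprint for item-received header sub-widgets, derived from the native ItemReceivedHeaderBase, with no visual widgets of its own.
# SetBlueprintGiftBoxItem copies the gift box item, sender name, received loot list and self-gift flag into Blueprint variables.
# An InitFromGiftBoxItem_BP event is declared but has no logic here.
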WidgetBlueprint /Game/UI/ItemReceived/ItemReceivedHeaders/ItemReceivedHeaderSubWidgetBase
graphs [
 EventGraph {nodes [Event InitFromGiftBoxItem_BP() {is=title K2Node_Event}]}
 SetBlueprintGiftBoxItem {nodes [
   n1 = SetBlueprintGiftBoxItem!() -> (
    then
    GiftBox:Fort Gift Box Item Object Reference
    FromUser:String
    ItemDefs:Array of Fort Received Item Loot Info Structures
    bSelfGift:bool
   ) {is=title K2Node_FunctionEntry}
   > Set GiftBoxItem_BP(GiftBoxItem_BP:Fort Gift Box Item Object Reference=@n1) -> (Output_Get:Fort Gift Box Item Object Reference)
   > Set FromUserName_BP(FromUserName_BP:String=@n1.FromUser) -> (Output_Get:String)
   > Set ItemDefs(ItemDefs:Array of Fort Received Item Loot Info Structures=@n1.ItemDefs) -> (Output_Get:Array of Fort Received Item Loot Info Structures)
   > Set bSelfGift(bSelfGift:bool=@n1.bSelfGift) -> (Output_Get:bool)
   > Return Node()
  ]}
]
variables [
 bSelfGift {type=boolean}
 FromUserName_BP {type=string, default=""}
 GiftBoxItem_BP {type=FortGiftBoxItem, default=null}
 ItemDefs {type="array<FortReceivedItemLootInfo>", default []}
]
widgets {info {inheritedWidgetCount=0, parentClass=/Script/FortniteUI.ItemReceivedHeaderBase, rootWidgetClass=null, widgetCount=0}, widgets []}
