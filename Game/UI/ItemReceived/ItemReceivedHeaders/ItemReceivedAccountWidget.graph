WidgetBlueprint /Game/UI/ItemReceived/ItemReceivedHeaders/ItemReceivedAccountWidget
graphs [EventGraph {nodes [
   n1 = Event OnAccountInfoChanged!() -> (OutputDelegate, then, Result:"Fort Public Account Info Structure (by ref)") {is=title K2Node_Event}
   n2 = Break Fort Public Account Info!(FortPublicAccountInfo:"Fort Public Account Info Structure (by ref)"=@n1.Result) -> (Level:int, MaxLevel:int, LevelXp:int, LevelXpForLevel:int, BattleRoyaleLevel:Athena Level Info Structure) {is=title K2Node_BreakStruct}
   > "SetText (Text)"(
    self=Get CommonTextBlock_UserMessage() -> (CommonTextBlock_UserMessage:Common Text Object Reference)
    InText=Format Text!(
     Format="LOCTEXT(\"You\\'ve reached tier {level}!\")"
     level:int=Break Athena Level Info!(AthenaLevelInfo:"Athena Level Info Structure (by ref)"=@n2.BattleRoyaleLevel) -> (AccountLevel:int, Level:int, MaxLevel:int, LevelXp:int, LevelXpForLevel:int, BookMaxLevel:int, BookLevelXpForLevel:int) {is=title K2Node_BreakStruct}.Level
    ) -> (Result)
   )
  ]}]
variables []
widgets {
 info {inheritedWidgetCount=0, parentClass=/Script/FortniteUI.FortAccountWidgetBase, rootWidgetClass=&VerticalBox, widgetCount=2}
 widgets [
  VerticalBox_0 {widgetClassPath=&VerticalBox, widget=VerticalBox_0}
  CommonTextBlock_UserMessage {
   widgetClassPath=&CommonTextBlock
   bIsVariable=true
   parent=VerticalBox_0
   widget=CommonTextBlock_UserMessage
   properties {
    style=/FortUILibrary/Styles/Typography/_TextStyle_SubH4._TextStyle_SubH4_C
    text="[PH] You've reached tier 40!"
    wrapTextAt=850
    font {
     bForceMonospaced=false
     bMaterialIsStencil=false
     fontMaterial=null
     fontObject=&BurbankSmall
     letterSpacing=0
     monospacedWidth=1
     size=24
     skewAmount=0
     typefaceFontName=Black
     outlineSettings {
      bApplyOutlineToDropShadows=false
      bMiteredCorners=false
      bSeparateFillAlpha=false
      outlineMaterial=null
      outlineSize=0
      outlineColor=<rgba 0 0 0 1>
     }
    }
    shadowOffset=<0 0>
    strikeBrush {
     bIsDynamicallyLoaded=false
     drawAs=NoDrawType
     imageType=NoImage
     mirroring=NoMirror
     resourceName=None
     resourceObject=null
     tiling=NoTile
     imageSize=<32 32>
     margin=<ltrb 0 0 0 0>
     outlineSettings {
      bUseBrushTransparency=false
      roundingType=HalfHeightRadius
      width=0
      color {colorUseRule=UseColor_Specified, specifiedColor=<rgba 0 0 0 0>}
      cornerRadii=<0 0 0 0>
     }
     tintColor {colorUseRule=UseColor_Specified, specifiedColor=<rgba 1 1 1 1>}
     uVRegion {bIsValid=false, max=<0 0>, min=<0 0>}
    }
   }
   slot {class=VerticalBoxSlot}
  }
 ]
}
