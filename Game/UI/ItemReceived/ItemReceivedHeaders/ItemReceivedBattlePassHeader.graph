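# Battle Pass header sub-widget for the item-received screen, showing an icon, a title, an account widget and an unlocked-item count.
# On InitFromGiftBoxItem_BP it picks the gold or silver icon from the Athena profile's HasPurchasedSeasonBook.
# It shows the "UPGRADED!" title when the gift box definition's name equals GB_BattlePassPurchased, otherwise "TIER UP!".
# The count text is plural-formatted from the ItemDefs array length.
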WidgetBlueprint /Game/UI/ItemReceived/ItemReceivedHeaders/ItemReceivedBattlePassHeader
graphs [EventGraph {nodes [
   Event InitFromGiftBoxItem_BP() {is=title K2Node_Event}
   > Parent: InitFromGiftBoxItem_BP() {is=title K2Node_CallParentFunction}
   n1 = GetAthenaProfile!(self:Fort Mcp Context Object Reference=Get FortMcpContext!() -> (ReturnValue:Fort Mcp Context Object Reference) {is=title K2Node_GetBlueprintContextEx}) -> (ReturnValue:Fort Mcp Profile Athena Object Reference)
   > SetActiveWidget(
    self:Widget Switcher Object Reference=Get CommonWidgetSwitcher_Icon() -> (CommonWidgetSwitcher_Icon:Common Widget Switcher Legacy Object Reference)
    Widget=Select(
     Option 0:Image Object Reference=Get ImageBattlePassIconSilver() -> (ImageBattlePassIconSilver:Image Object Reference)
     Option 1:Image Object Reference=Get ImageBattlePassIconGold() -> (ImageBattlePassIconGold:Image Object Reference)
     Index=HasPurchasedSeasonBook!(self:Fort Mcp Profile Athena Object Reference=@n1) -> (ReturnValue:bool)
    ) -> (ReturnValue:Image Object Reference)
   )
   n2 = GetItemDefinition!(self:Fort Item Object Reference=Get GiftBoxItem_BP() -> (GiftBoxItem_BP:Fort Gift Box Item Object Reference)) -> (ReturnValue:Fort Item Definition Object Reference)
   > Set Is BattlePass(Is BattlePass:bool="Equal Exactly (String)"(A=GetObjectName(Object=@n2), B=GB_BattlePassPurchased)) -> (Output_Get:bool)
   > SetActiveWidget(
    self=Get CommonWidgetSwitcher_Title() -> (CommonWidgetSwitcher_Title:Common Widget Switcher Legacy Object Reference)
    Widget=Select(
     Option 0:Common Text Object Reference=Get BattlepassTier() -> (BattlepassTier:Common Text Object Reference)
     Option 1:Common Text Object Reference=Get BattlepassUpgraded() -> (BattlepassUpgraded:Common Text Object Reference)
     Index=Get Is BattlePass() -> (Is BattlePass:bool)
    ) -> (ReturnValue:Common Text Object Reference)
   )
   n3 = Get ItemDefs() -> (ItemDefs:Array of Fort Received Item Loot Info Structures)
   > "SetText (Text)"(
    self=Get CommonTextBlock_ItemCount() -> (CommonTextBlock_ItemCount:Common Text Object Reference)
    InText=Format Text!(
     Format="LOCTEXT(\"{Number} {Number}|plural(one=item has,other=items have) been unlocked!\")"
     Number:int=Length(TargetArray:Array of Fort Received Item Loot Info Structures=@n3)
    ) -> (Result)
   )
  ]}]
variables [Is BattlePass {type=boolean}]
widgets {
 info {
  inheritedWidgetCount=0
  parentClass=&/ItemReceivedHeaderSubWidgetBase.ItemReceivedHeaderSubWidgetBase_C
  rootWidgetClass=&HorizontalBox
  widgetCount=13
 }
 widgets [
  HorizontalBox_1 {widgetClassPath=&HorizontalBox, widget=HorizontalBox_1}
  HorizontalBox_0 {
   widgetClassPath=&HorizontalBox
   parent=HorizontalBox_1
   widget=HorizontalBox_0
   slot {class=HorizontalBoxSlot, properties {horizontalAlignment=HAlign_Left, verticalAlignment=VAlign_Bottom}}
  }
  CommonWidgetSwitcher_Icon {
   widgetClassPath=&CommonWidgetSwitcherLegacy
   bIsVariable=true
   parent=HorizontalBox_0
   widget=CommonWidgetSwitcher_Icon
   slot {class=HorizontalBoxSlot, properties {horizontalAlignment=HAlign_Left, verticalAlignment=VAlign_Center}}
  }
  ImageBattlePassIconSilver {
   widgetClassPath=&Image
   bIsVariable=true
   parent=CommonWidgetSwitcher_Icon
   widget=ImageBattlePassIconSilver
   properties {brush {
     bIsDynamicallyLoaded=false
     drawAs=Image
     imageType=NoImage
     mirroring=NoMirror
     resourceName=None
     resourceObject=/Game/UI/Foundation/Textures/Icons/Items/T-FNBR-BattlePass-Default-L
     tiling=NoTile
     imageSize=<256 256>
     margin=<ltrb 0 0 0 0>
     outlineSettings {
      bUseBrushTransparency=false
      roundingType=HalfHeightRadius
      width=0
      color {colorUseRule=UseColor_Specified, specifiedColor=<rgba 0 0 0 0>}
      cornerRadii=<0 0 0 0>
     }
     tintColor {colorUseRule=UseColor_Specified, specifiedColor=<rgba 1 1 1 1>}
     uVRegion {bIsValid=false, max=<0 0>, min=<0 0>}
    }}
   slot {class=WidgetSwitcherSlot}
  }
  ImageBattlePassIconGold {
   widgetClassPath=&Image
   bIsVariable=true
   parent=CommonWidgetSwitcher_Icon
   widget=ImageBattlePassIconGold
   properties {brush {
     bIsDynamicallyLoaded=false
     drawAs=Image
     imageType=NoImage
     mirroring=NoMirror
     resourceName=None
     resourceObject=/Game/UI/Foundation/Textures/Icons/Items/T-FNBR-BattlePass-L
     tiling=NoTile
     imageSize=<256 256>
     margin=<ltrb 0 0 0 0>
     outlineSettings {
      bUseBrushTransparency=false
      roundingType=HalfHeightRadius
      width=0
      color {colorUseRule=UseColor_Specified, specifiedColor=<rgba 0 0 0 0>}
      cornerRadii=<0 0 0 0>
     }
     tintColor {colorUseRule=UseColor_Specified, specifiedColor=<rgba 1 1 1 1>}
     uVRegion {bIsValid=false, max=<0 0>, min=<0 0>}
    }}
   slot {class=WidgetSwitcherSlot}
  }
  VerticalBox_2 {
   widgetClassPath=&VerticalBox
   parent=HorizontalBox_0
   widget=VerticalBox_2
   slot {class=HorizontalBoxSlot, properties {horizontalAlignment=HAlign_Left, verticalAlignment=VAlign_Center, padding=<ltrb 50 0 50 0>}}
  }
  BorderItemNAme {
   widgetClassPath=&Border
   parent=VerticalBox_2
   widget=BorderItemNAme
   properties {
    verticalAlignment=VAlign_Center
    background {
     bIsDynamicallyLoaded=false
     drawAs=NoDrawType
     imageType=NoImage
     mirroring=NoMirror
     resourceName=None
     resourceObject=null
     tiling=NoTile
     imageSize=<32 32>
     margin=<ltrb 0 0 0 0>
     outlineSettings {
      bUseBrushTransparency=false
      roundingType=HalfHeightRadius
      width=0
      color {colorUseRule=UseColor_Specified, specifiedColor=<rgba 0 0 0 0>}
      cornerRadii=<0 0 0 0>
     }
     tintColor {colorUseRule=UseColor_Specified, specifiedColor=<rgba 1 1 1 1>}
     uVRegion {bIsValid=false, max=<0 0>, min=<0 0>}
    }
    padding=<ltrb 0 -12 0 -12>
   }
   slot {class=VerticalBoxSlot}
  }
  CommonWidgetSwitcher_Title {
   widgetClassPath=&CommonWidgetSwitcherLegacy
   bIsVariable=true
   parent=BorderItemNAme
   widget=CommonWidgetSwitcher_Title
   slot {class=BorderSlot, properties {verticalAlignment=VAlign_Center, padding=<ltrb 0 -12 0 -12>}}
  }
  BattlepassUpgraded {
   widgetClassPath=&CommonTextBlock
   bIsVariable=true
   parent=CommonWidgetSwitcher_Title
   widget=BattlepassUpgraded
   properties {
    justification=Center
    scrollStyle=&TextScrollStyle-Base_C
    style=/FortUILibrary/Styles/Typography/Foundation/TextStyle-Header-XL-Horizon.TextStyle-Header-XL-Horizon_C
    text="Battle Pass UPGRADED!"
    font {
     bForceMonospaced=false
     bMaterialIsStencil=false
     fontMaterial=/FortUILibrary/Materials/Foundation/M_Font_Horizon_Dynamic
     fontObject=&Burbank
     letterSpacing=0
     monospacedWidth=1
     size=72
     skewAmount=0
     typefaceFontName=Burbank-Black
     outlineSettings {
      bApplyOutlineToDropShadows=true
      bMiteredCorners=false
      bSeparateFillAlpha=false
      outlineMaterial=null
      outlineSize=3
      outlineColor=<rgba 0 0 0 0.1>
     }
    }
    shadowOffset=<0 0>
   }
   slot {class=WidgetSwitcherSlot}
  }
  BattlepassTier {
   widgetClassPath=&CommonTextBlock
   bIsVariable=true
   parent=CommonWidgetSwitcher_Title
   widget=BattlepassTier
   properties {
    justification=Center
    scrollStyle=&TextScrollStyle-Base_C
    style=&TextStyle-Header-XL-Horizon_C
    text="Battle Pass TIER UP!"
    font {
     bForceMonospaced=false
     bMaterialIsStencil=false
     fontMaterial=&M_Font_Horizon_Dynamic
     fontObject=&Burbank
     letterSpacing=0
     monospacedWidth=1
     size=72
     skewAmount=0
     typefaceFontName=Burbank-Black
     outlineSettings {
      bApplyOutlineToDropShadows=true
      bMiteredCorners=false
      bSeparateFillAlpha=false
      outlineMaterial=null
      outlineSize=3
      outlineColor=<rgba 0 0 0 0.1>
     }
    }
    shadowOffset=<0 0>
   }
   slot {class=WidgetSwitcherSlot}
  }
  VerticalBox_0 {
   widgetClassPath=&VerticalBox
   parent=VerticalBox_2
   widget=VerticalBox_0
   slot {class=VerticalBoxSlot, properties {verticalAlignment=VAlign_Bottom}}
  }
  ItemReceivedAccountWidget {
   widgetClassPath=&/ItemReceivedAccountWidget
   bIsVariable=true
   parent=VerticalBox_0
   widget=ItemReceivedAccountWidget
   slot {class=VerticalBoxSlot}
  }
  CommonTextBlock_ItemCount {
   widgetClassPath=&CommonTextBlock
   bIsVariable=true
   parent=VerticalBox_0
   widget=CommonTextBlock_ItemCount
   properties {
    style=/FortUILibrary/Styles/Typography/Foundation/TextStyle-Base-M.TextStyle-Base-M_C
    text="[ph]X items have been unlocked!"
    wrapTextAt=600
    font {
     bForceMonospaced=false
     bMaterialIsStencil=false
     fontMaterial=null
     fontObject=&NotoSans
     letterSpacing=0
     monospacedWidth=1
     size=24
     skewAmount=0
     typefaceFontName=NotoSans-Regular
     outlineSettings {
      bApplyOutlineToDropShadows=false
      bMiteredCorners=false
      bSeparateFillAlpha=false
      outlineMaterial=null
      outlineSize=0
      outlineColor=<rgba 0 0 0 1>
     }
    }
    shadowOffset=<0 0>
   }
   slot {class=VerticalBoxSlot}
  }
 ]
}
