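# Item-received header widget that announces a gift-box reward for completing a challenge bundle, with a Battle Pass icon beside title and message text.
# On InitFromGiftBoxItem_BP it calls the parent, then switches between silver and gold icons by whether the Athena profile has purchased the season book.
# It casts the item definition from the HeaderTemplateId param to a challenge bundle and formats its name into the message.
# The Number argument is unconnected, and the title switcher is never set here.
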WidgetBlueprint /Game/UI/ItemReceived/ItemReceivedHeaders/ItemReceivedChallengeCompleteHeader
graphs [EventGraph {nodes [
   Event InitFromGiftBoxItem_BP() {is=title K2Node_Event}
   > Parent: InitFromGiftBoxItem_BP() {is=title K2Node_CallParentFunction}
   n1 = GetAthenaProfile!(self:Fort Mcp Context Object Reference=Get FortMcpContext!() -> (ReturnValue:Fort Mcp Context Object Reference) {is=title K2Node_GetBlueprintContextEx}) -> (ReturnValue:Fort Mcp Profile Athena Object Reference)
   > SetActiveWidget(
    self:Widget Switcher Object Reference=Get CommonWidgetSwitcher_Icon() -> (CommonWidgetSwitcher_Icon:Common Widget Switcher Legacy Object Reference)
    Widget=Select(
     Option 0:Image Object Reference=Get ImageBattlePassIconSilver() -> (ImageBattlePassIconSilver:Image Object Reference)
     Option 1:Image Object Reference=Get ImageBattlePassIconGold() -> (ImageBattlePassIconGold:Image Object Reference)
     Index=HasPurchasedSeasonBook!(self:Fort Mcp Profile Athena Object Reference=@n1) -> (ReturnValue:bool)
    ) -> (ReturnValue:Image Object Reference)
   )
   > n2 = Cast To FortChallengeBundleItemDefinition!(
    execute
    Object=RetrieveItemDefintionFromParam!(
     self:Fort Gift Box Item Object Reference=Get GiftBoxItem_BP() -> (GiftBoxItem_BP:Fort Gift Box Item Object Reference)
     ParamName:String=HeaderTemplateId
    ) -> (ReturnValue:Fort Item Definition Object Reference)
   ) -> (then, CastFailed, AsFort Challenge Bundle Item Definition:Fort Challenge Bundle Item Definition Object Reference) {is=title K2Node_DynamicCast}
   > "SetText (Text)"(
    self=Get CommonTextBlock_2() -> (CommonTextBlock_2:Common Text Object Reference)
    InText=Format Text!(
     Format="LOCTEXT(\"You have received a reward for completing the {Bundle Name} {Number}challenges!\")"
     Bundle Name:Text=Get ItemName!(self:Item Definition Base Object Reference=@n2) -> (ItemName:Text)
     Number:Wildcard
    ) -> (Result)
   )
   Select(Option 0:Text=Format Text!(Format="LOCTEXT(\"{x} \")", x:Wildcard) -> (Result), Option 1:Text, Index=TextIsEmpty()) -> (ReturnValue:Text)
  ]}]
variables []
widgets {
 info {
  inheritedWidgetCount=0
  parentClass=&/ItemReceivedHeaderSubWidgetBase.ItemReceivedHeaderSubWidgetBase_C
  rootWidgetClass=&HorizontalBox
  widgetCount=13
 }
 widgets [
  HorizontalBox_1 {widgetClassPath=&HorizontalBox, widget=HorizontalBox_1}
  HorizontalBox_0 {
   widgetClassPath=&HorizontalBox
   parent=HorizontalBox_1
   widget=HorizontalBox_0
   slot {class=HorizontalBoxSlot, properties {horizontalAlignment=HAlign_Left, verticalAlignment=VAlign_Bottom}}
  }
  CommonWidgetSwitcher_Icon {
   widgetClassPath=&CommonWidgetSwitcherLegacy
   bIsVariable=true
   parent=HorizontalBox_0
   widget=CommonWidgetSwitcher_Icon
   slot {class=HorizontalBoxSlot, properties {horizontalAlignment=HAlign_Left, verticalAlignment=VAlign_Center}}
  }
  ImageBattlePassIconSilver {
   widgetClassPath=&Image
   bIsVariable=true
   parent=CommonWidgetSwitcher_Icon
   widget=ImageBattlePassIconSilver
   properties {brush {
     bIsDynamicallyLoaded=false
     drawAs=Image
     imageType=NoImage
     mirroring=NoMirror
     resourceName=None
     resourceObject=/Game/UI/Foundation/Textures/Icons/Items/T-FNBR-BattlePass-Default-L
     tiling=NoTile
     imageSize=<256 256>
     margin=<ltrb 0 0 0 0>
     outlineSettings {
      bUseBrushTransparency=false
      roundingType=HalfHeightRadius
      width=0
      color {colorUseRule=UseColor_Specified, specifiedColor=<rgba 0 0 0 0>}
      cornerRadii=<0 0 0 0>
     }
     tintColor {colorUseRule=UseColor_Specified, specifiedColor=<rgba 1 1 1 1>}
     uVRegion {bIsValid=false, max=<0 0>, min=<0 0>}
    }}
   slot {class=WidgetSwitcherSlot}
  }
  ImageBattlePassIconGold {
   widgetClassPath=&Image
   bIsVariable=true
   parent=CommonWidgetSwitcher_Icon
   widget=ImageBattlePassIconGold
   properties {brush {
     bIsDynamicallyLoaded=false
     drawAs=Image
     imageType=NoImage
     mirroring=NoMirror
     resourceName=None
     resourceObject=/Game/UI/Foundation/Textures/Icons/Items/T-FNBR-BattlePass-L
     tiling=NoTile
     imageSize=<256 256>
     margin=<ltrb 0 0 0 0>
     outlineSettings {
      bUseBrushTransparency=false
      roundingType=HalfHeightRadius
      width=0
      color {colorUseRule=UseColor_Specified, specifiedColor=<rgba 0 0 0 0>}
      cornerRadii=<0 0 0 0>
     }
     tintColor {colorUseRule=UseColor_Specified, specifiedColor=<rgba 1 1 1 1>}
     uVRegion {bIsValid=false, max=<0 0>, min=<0 0>}
    }}
   slot {class=WidgetSwitcherSlot}
  }
  VerticalBox_2 {
   widgetClassPath=&VerticalBox
   parent=HorizontalBox_0
   widget=VerticalBox_2
   slot {class=HorizontalBoxSlot, properties {horizontalAlignment=HAlign_Left, verticalAlignment=VAlign_Center, padding=<ltrb 50 0 50 0>}}
  }
  BorderItemNAme {
   widgetClassPath=&Border
   parent=VerticalBox_2
   widget=BorderItemNAme
   properties {
    background {
     bIsDynamicallyLoaded=false
     drawAs=NoDrawType
     imageType=NoImage
     mirroring=NoMirror
     resourceName=None
     resourceObject=null
     tiling=NoTile
     imageSize=<32 32>
     margin=<ltrb 0 0 0 0>
     outlineSettings {
      bUseBrushTransparency=false
      roundingType=HalfHeightRadius
      width=0
      color {colorUseRule=UseColor_Specified, specifiedColor=<rgba 0 0 0 0>}
      cornerRadii=<0 0 0 0>
     }
     tintColor {colorUseRule=UseColor_Specified, specifiedColor=<rgba 1 1 1 1>}
     uVRegion {bIsValid=false, max=<0 0>, min=<0 0>}
    }
    padding=<ltrb 0 0 0 0>
   }
   slot {class=VerticalBoxSlot}
  }
  CommonWidgetSwitcher_Title {
   widgetClassPath=&CommonWidgetSwitcherLegacy
   bIsVariable=true
   parent=BorderItemNAme
   widget=CommonWidgetSwitcher_Title
   slot {class=BorderSlot, properties {padding=<ltrb 0 0 0 0>}}
  }
  BattlepassUpgraded {
   widgetClassPath=&CommonTextBlock
   bIsVariable=true
   parent=CommonWidgetSwitcher_Title
   widget=BattlepassUpgraded
   properties {
    justification=Center
    scrollStyle=&TextScrollStyle-Base_C
    text="Congratulations!"
    textTransformPolicy=ToUpper
    font {
     bForceMonospaced=false
     bMaterialIsStencil=false
     fontMaterial=null
     fontObject=&Burbank
     letterSpacing=0
     monospacedWidth=1
     size=72
     skewAmount=0
     typefaceFontName=Burbank-Black
     outlineSettings {
      bApplyOutlineToDropShadows=false
      bMiteredCorners=false
      bSeparateFillAlpha=false
      outlineMaterial=null
      outlineSize=0
      outlineColor=<rgba 0 0 0 1>
     }
    }
    shadowOffset=<0 0>
   }
   slot {class=WidgetSwitcherSlot}
  }
  BattlepassTier {
   widgetClassPath=&CommonTextBlock
   bIsVariable=true
   parent=CommonWidgetSwitcher_Title
   widget=BattlepassTier
   properties {
    justification=Center
    scrollStyle=&TextScrollStyle-Base_C
    text="Battle Pass TIER UP!"
    textTransformPolicy=ToUpper
    font {
     bForceMonospaced=false
     bMaterialIsStencil=false
     fontMaterial=null
     fontObject=&Burbank
     letterSpacing=0
     monospacedWidth=1
     size=72
     skewAmount=0
     typefaceFontName=Burbank-Black
     outlineSettings {
      bApplyOutlineToDropShadows=false
      bMiteredCorners=false
      bSeparateFillAlpha=false
      outlineMaterial=null
      outlineSize=0
      outlineColor=<rgba 0 0 0 1>
     }
    }
    shadowOffset=<0 0>
   }
   slot {class=WidgetSwitcherSlot}
  }
  VerticalBox_0 {widgetClassPath=&VerticalBox, parent=VerticalBox_2, widget=VerticalBox_0, slot {class=VerticalBoxSlot}}
  ItemReceivedAccountWidget {
   widgetClassPath=&/ItemReceivedAccountWidget
   bIsVariable=true
   parent=VerticalBox_0
   widget=ItemReceivedAccountWidget
   properties {visibility=Collapsed}
   slot {class=VerticalBoxSlot}
  }
  CommonTextBlock_2 {
   widgetClassPath=&CommonTextBlock
   bIsVariable=true
   parent=VerticalBox_0
   widget=CommonTextBlock_2
   properties {
    style=/FortUILibrary/Styles/Typography/Foundation/TextStyle-Base-M.TextStyle-Base-M_C
    text="\"You have received a reward for completing the {Bundle Name} {Number} quests!"
    wrapTextAt=600
    font {
     bForceMonospaced=false
     bMaterialIsStencil=false
     fontMaterial=null
     fontObject=&NotoSans
     letterSpacing=0
     monospacedWidth=1
     size=24
     skewAmount=0
     typefaceFontName=NotoSans-Regular
     outlineSettings {
      bApplyOutlineToDropShadows=false
      bMiteredCorners=false
      bSeparateFillAlpha=false
      outlineMaterial=null
      outlineSize=0
      outlineColor=<rgba 0 0 0 1>
     }
    }
    shadowOffset=<0 0>
   }
   slot {class=VerticalBoxSlot, properties {horizontalAlignment=HAlign_Left}}
  }
 ]
}
