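# Item-received header widget for gift boxes from Epic Games, such as make-goods, migrations and rewards, showing the Epic logo and name.
# On InitFromGiftBoxItem_BP it calls the parent, then shows the formatted userMessage param, or collapses that text when the param is absent.
# It also resolves the item's assetString param as a key into the AssetStrings map of canned messages, collapsing that text when no entry matches.
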WidgetBlueprint /Game/UI/ItemReceived/ItemReceivedHeaders/ItemReceivedEpicMakeGoodHeader
graphs [EventGraph {nodes [
   Event InitFromGiftBoxItem_BP() {is=title K2Node_Event}
   > Parent: InitFromGiftBoxItem_BP() {is=title K2Node_CallParentFunction}
   > n1 = Sequence()
   n2 = MakeLiteralText(Value="LOCTEXT(\"{assetString}\")")
   n3 = Get GiftBoxItem_BP() -> (GiftBoxItem_BP:Fort Gift Box Item Object Reference)
   n4 = Find(
    TargetMap:Map of Strings to Texts=Get AssetStrings() -> (AssetStrings:Map of Strings to Texts)
    Key="To String (Text)"(InText=FormatTextUsingParams!(self:Fort Gift Box Item Object Reference=@n3, FormatString:"Text (by ref)"=@n2) -> (ReturnValue:Text))
   ) -> (Value:Text)
   n5 = Branch(execute=@n1.then_1, Condition=@n4.ReturnValue)
   > "SetText (Text)"(self=Get AssetString() -> (AssetString:Common Text Object Reference), InText=@n4)
   SetVisibility(execute=@n5.else, self=Get AssetString(), InVisibility=Collapsed)
   n6 = Get GiftBoxItem_BP()
   n7 = Branch(
    execute=@n1.then_0
    Condition=NOT Boolean(A=HasParam!(self:Fort Gift Box Item Object Reference=@n6, ParamName:String=userMessage) -> (ReturnValue:bool))
   )
   > SetVisibility(self=Get UserMessage() -> (UserMessage:Common Text Object Reference), InVisibility=Collapsed)
   "SetText (Text)"(
    execute=@n7.else
    self=Get UserMessage()
    InText=FormatTextUsingParams(self=@n6, FormatString=MakeLiteralText(Value="LOCTEXT(\"{userMessage}\")"))
   )
  ]}]
variables [AssetStrings {
  type="map<string>"
  default {
   Compensation.Downtime="We apologize for the recent downtime, so we want you to have this. Good luck; have fun!"
   Inventory.Removed=The following items were removed from your account by customer support:
   Migration.AddedItems="One of your cosmetic items now includes additional items. Enjoy!"
   Migration.AddedItems.BackBling="One of your cosmetic items now includes a matching Back Bling. Enjoy!"
   Migration.CollectionBookRewards=The Collection Book is now more rewarding. Here are the updated rewards for everything you've added to the collection book so far.
   Migration.FounderChatMakeGood="Thanks for being a Save the World Super Deluxe, Limited, or Ultimate edition Founder!"
   Migration.General="Some rewards are now given out differently, so you deserve this. Thanks for playing!"
   Migration.HeroLoadout="Here are some things to help you get started in the new Hero Loadout!"
   Migration.InventoryBonus="We've changed how some items are stored in your inventory, so we want you to have this. Enjoy!"
   Migration.JacksRevengeAlterationsMakeGood="Due to a bug, some of the perks on the Jack's Revenge pistol have been reset. Sorry for the inconvenience!"
   Migration.MakeGood=Mistakes were made. We apologize for the missed rewards.
   Migration.RecyclingMakeGood="Oops! We dropped these.  Here are some missing recycled materials."
   Migration.RetroactivePostMaxAccountLevelRewards="We started keeping track of any XP acquired beyond the current max level back in 6.30. Here are your rewards!"
   Migration.SSD10Loadout_T1="Here's a reward for completing the final Storm Shield Defense in Stonewood!"
   Migration.SSD10Loadout_T2="Here's a reward for completing the final Storm Shield Defense in Plankerton!"
   Migration.SSD10Loadout_T3="Here's a reward for completing the final Storm Shield Defense in Canny Valley!"
   Migration.SSD10Loadout_T4="Here's a reward for completing the final Storm Shield Defense in Twine Peaks!"
   Migration.SkillTree="We have removed the Skill Tree and introduced the new Upgrades Page. Here are your new Upgrade Points, F.O.R.T. stats and many other rewards!"
   Reward.Generic="There could be any number of reasons why you're getting this item. Enjoy!"
   Reward.GenericGift1="A gift from Epic. Enjoy!"
   Reward.GenericGift2="Loot! Enjoy this gift from Epic"
   Reward.GenericGift3="A new item for your collection. Enjoy!"
   Reward.GenericGift4="Well look who's got themselves some more goodness!"
   Reward.GenericGift5="Here is your reward. Enjoy!"
   Reward.LimitedTimeMashup="Thank you for playing the recent Limited Time Mashup. Enjoy!"
   Reward.LimitedTimeMode="Thank you for playing the recent Limited Time Mode. Enjoy!"
   Reward.MarshmelloLS="Due to the issues last week with the Marshmello Outfit, we're granting you a special loading screen from the event. Thanks for your patience!"
   Reward.QuestsAndMissions="Here are your Rewards from previously completed Missions and Quests!"
   Reward.ReachedSeasonLevel100="Congratulations, you reached Season Level 100. Enjoy!"
   Reward.Showdown.HighPlacement="Congratulations on your excellent performance in the Showdown LTM!"
   Reward.SoloShowdown.HighPlacement="Congratulations on your excellent performance in the Solo Showdown LTM!"
   Reward.SoloShowdownV1.Played50Games="Congratulations, you completed 50 games in the Solo Showdown LTM! As a thank you for participating, you have been awarded the unique Showdown Spray."
   Seasonal.Autoclaim=These were collected for you from last season.
  }
 }]
widgets {
 info {
  inheritedWidgetCount=0
  parentClass=&/ItemReceivedHeaderSubWidgetBase.ItemReceivedHeaderSubWidgetBase_C
  rootWidgetClass=&VerticalBox
  widgetCount=12
 }
 widgets [
  VerticalBox_1 {widgetClassPath=&VerticalBox, widget=VerticalBox_1}
  VerticalBox_0 {
   widgetClassPath=&VerticalBox
   parent=VerticalBox_1
   widget=VerticalBox_0
   slot {class=VerticalBoxSlot, properties {horizontalAlignment=HAlign_Center}}
  }
  HorizontalBox_0 {widgetClassPath=&HorizontalBox, parent=VerticalBox_0, widget=HorizontalBox_0, slot {class=VerticalBoxSlot}}
  Image_0 {
   widgetClassPath=&Image
   bIsVariable=true
   parent=HorizontalBox_0
   widget=Image_0
   properties {brush {
     bIsDynamicallyLoaded=false
     drawAs=Image
     imageType=NoImage
     mirroring=NoMirror
     resourceName=None
     resourceObject=/Game/UI/Foundation/Textures/Icons/Social/T-Icon-EpicGamesLogo-64
     tiling=NoTile
     imageSize=<64 64>
     margin=<ltrb 0 0 0 0>
     outlineSettings {
      bUseBrushTransparency=false
      roundingType=HalfHeightRadius
      width=0
      color {colorUseRule=UseColor_Specified, specifiedColor=<rgba 0 0 0 0>}
      cornerRadii=<0 0 0 0>
     }
     tintColor {colorUseRule=UseColor_Specified, specifiedColor=<rgba 1 1 1 1>}
     uVRegion {bIsValid=false, max=<0 0>, min=<0 0>}
    }}
   slot {class=HorizontalBoxSlot, properties {horizontalAlignment=HAlign_Center, verticalAlignment=VAlign_Center}}
  }
  BorderItemNAme {
   widgetClassPath=&Border
   parent=HorizontalBox_0
   widget=BorderItemNAme
   properties {
    background {
     bIsDynamicallyLoaded=false
     drawAs=NoDrawType
     imageType=NoImage
     mirroring=NoMirror
     resourceName=None
     resourceObject=null
     tiling=NoTile
     imageSize=<32 32>
     margin=<ltrb 0 0 0 0>
     outlineSettings {
      bUseBrushTransparency=false
      roundingType=HalfHeightRadius
      width=0
      color {colorUseRule=UseColor_Specified, specifiedColor=<rgba 0 0 0 0>}
      cornerRadii=<0 0 0 0>
     }
     tintColor {colorUseRule=UseColor_Specified, specifiedColor=<rgba 1 1 1 1>}
     uVRegion {bIsValid=false, max=<0 0>, min=<0 0>}
    }
    padding=<ltrb 0 0 0 0>
   }
   slot {class=HorizontalBoxSlot}
  }
  Overlay_9 {widgetClassPath=&Overlay, parent=BorderItemNAme, widget=Overlay_9, slot {class=BorderSlot, properties {padding=<ltrb 0 0 0 0>}}}
  TextName {
   widgetClassPath=&CommonTextBlock
   bIsVariable=true
   parent=Overlay_9
   widget=TextName
   properties {
    justification=Center
    scrollStyle=&TextScrollStyle-Base_C
    style=/FortUILibrary/Styles/Typography/Foundation/TextStyle-Header-XL-Horizon.TextStyle-Header-XL-Horizon_C
    text=Epic Games
    font {
     bForceMonospaced=false
     bMaterialIsStencil=false
     fontMaterial=/FortUILibrary/Materials/Foundation/M_Font_Horizon_Dynamic
     fontObject=&Burbank
     letterSpacing=0
     monospacedWidth=1
     size=72
     skewAmount=0
     typefaceFontName=Burbank-Black
     outlineSettings {
      bApplyOutlineToDropShadows=true
      bMiteredCorners=false
      bSeparateFillAlpha=false
      outlineMaterial=null
      outlineSize=3
      outlineColor=<rgba 0 0 0 0.1>
     }
    }
    shadowOffset=<0 0>
   }
   slot {class=OverlaySlot, properties {padding=<ltrb 0 -10 0 -20>}}
  }
  BorderSuccessText {
   widgetClassPath=&Border
   parent=VerticalBox_0
   widget=BorderSuccessText
   properties {
    visibility=Collapsed
    background {
     bIsDynamicallyLoaded=false
     drawAs=NoDrawType
     imageType=NoImage
     mirroring=NoMirror
     resourceName=None
     resourceObject=null
     tiling=NoTile
     imageSize=<32 32>
     margin=<ltrb 0 0 0 0>
     outlineSettings {
      bUseBrushTransparency=false
      roundingType=HalfHeightRadius
      width=0
      color {colorUseRule=UseColor_Specified, specifiedColor=<rgba 0 0 0 0>}
      cornerRadii=<0 0 0 0>
     }
     tintColor {colorUseRule=UseColor_Specified, specifiedColor=<rgba 1 1 1 1>}
     uVRegion {bIsValid=false, max=<0 0>, min=<0 0>}
    }
    padding=<ltrb 0 0 0 0>
   }
   slot {class=VerticalBoxSlot, properties {horizontalAlignment=HAlign_Center}}
  }
  Sender {
   widgetClassPath=&CommonTextBlock
   bIsVariable=true
   parent=BorderSuccessText
   widget=Sender
   properties {
    justification=Center
    style=/FortUILibrary/Styles/Typography/Foundation/TextStyle-Header-M.TextStyle-Header-M_C
    text="Sent you a gift box!"
    font {
     bForceMonospaced=false
     bMaterialIsStencil=false
     fontMaterial=null
     fontObject=&Burbank
     letterSpacing=0
     monospacedWidth=1
     size=32
     skewAmount=0
     typefaceFontName=Burbank-Black
     outlineSettings {
      bApplyOutlineToDropShadows=false
      bMiteredCorners=false
      bSeparateFillAlpha=false
      outlineMaterial=null
      outlineSize=0
      outlineColor=<rgba 0 0 0 1>
     }
    }
    shadowOffset=<0 0>
   }
   slot {class=BorderSlot, properties {padding=<ltrb 0 0 0 0>}}
  }
  VerticalBox_Text {widgetClassPath=&VerticalBox, bIsVariable=true, parent=VerticalBox_1, widget=VerticalBox_Text, slot {class=VerticalBoxSlot}}
  UserMessage {
   widgetClassPath=&CommonTextBlock
   bIsVariable=true
   parent=VerticalBox_Text
   widget=UserMessage
   properties {
    autoWrapText=true
    justification=Center
    style=/FortUILibrary/Styles/Typography/Foundation/TextStyle-Base-M.TextStyle-Base-M_C
    text="{userMessage}"
    wrapTextAt=850
    font {
     bForceMonospaced=false
     bMaterialIsStencil=false
     fontMaterial=null
     fontObject=&NotoSans
     letterSpacing=0
     monospacedWidth=1
     size=24
     skewAmount=0
     typefaceFontName=NotoSans-Regular
     outlineSettings {
      bApplyOutlineToDropShadows=false
      bMiteredCorners=false
      bSeparateFillAlpha=false
      outlineMaterial=null
      outlineSize=0
      outlineColor=<rgba 0 0 0 1>
     }
    }
    shadowOffset=<0 0>
   }
   slot {class=VerticalBoxSlot, properties {horizontalAlignment=HAlign_Center}}
  }
  AssetString {
   widgetClassPath=&CommonTextBlock
   bIsVariable=true
   parent=VerticalBox_Text
   widget=AssetString
   properties {
    autoWrapText=true
    justification=Center
    style=&TextStyle-Base-M_C
    text="{AssetString} Lorem Ipsum Dolor Amit Lorem Ipsum Dolor Amit Lorem Ipsum Dolor Amit"
    wrapTextAt=850
    font {
     bForceMonospaced=false
     bMaterialIsStencil=false
     fontMaterial=null
     fontObject=&NotoSans
     letterSpacing=0
     monospacedWidth=1
     size=24
     skewAmount=0
     typefaceFontName=NotoSans-Regular
     outlineSettings {
      bApplyOutlineToDropShadows=false
      bMiteredCorners=false
      bSeparateFillAlpha=false
      outlineMaterial=null
      outlineSize=0
      outlineColor=<rgba 0 0 0 1>
     }
    }
    shadowOffset=<0 0>
   }
   slot {class=VerticalBoxSlot, properties {horizontalAlignment=HAlign_Center}}
  }
 ]
}
