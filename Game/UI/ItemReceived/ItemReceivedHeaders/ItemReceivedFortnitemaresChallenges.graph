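# Header sub-widget for the item-received screen that presents Fortnitemares rewards with a "FORTNITEMARES" title and "THANKS FOR PLAYING!" subheader.
# A width-limited SizeBox stacks both purple and magenta texts over a giftbox cube-pattern material.
# A translucent magenta soft-glow image draws on top, and negative padding lets both images bleed past the text.
# The event graph is empty, and the title keeps the name BattlepassUpgraded despite its Fortnitemares text.
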WidgetBlueprint /Game/UI/ItemReceived/ItemReceivedHeaders/ItemReceivedFortnitemaresChallenges
graphs [EventGraph {links [], nodes []}]
variables []
widgets {
 info {
  inheritedWidgetCount=0
  parentClass=&/ItemReceivedHeaderSubWidgetBase.ItemReceivedHeaderSubWidgetBase_C
  rootWidgetClass=&VerticalBox
  widgetCount=11
 }
 widgets [
  VerticalBox_1 {widgetClassPath=&VerticalBox, widget=VerticalBox_1}
  Overlay_1 {widgetClassPath=&Overlay, parent=VerticalBox_1, widget=Overlay_1, slot {class=VerticalBoxSlot}}
  Overlay_2 {
   widgetClassPath=&Overlay
   parent=Overlay_1
   widget=Overlay_2
   slot {class=OverlaySlot, properties {verticalAlignment=VAlign_Center}}
  }
  PrimaryGlow {
   widgetClassPath=&Image
   parent=Overlay_2
   widget=PrimaryGlow
   properties {
    visibility=SelfHitTestInvisible
    brush {
     bIsDynamicallyLoaded=false
     drawAs=Image
     imageType=NoImage
     mirroring=NoMirror
     resourceName=None
     resourceObject=/FortUILibrary/Styles/Border/Materials/M_UI_CubePattern_Giftbox_Inst
     tiling=NoTile
     imageSize=<512 64>
     margin=<ltrb 0.5 0 0.5 0>
     outlineSettings {
      bUseBrushTransparency=false
      roundingType=HalfHeightRadius
      width=0
      color {colorUseRule=UseColor_Specified, specifiedColor=<rgba 0 0 0 0>}
      cornerRadii=<0 0 0 0>
     }
     tintColor {colorUseRule=UseColor_Specified, specifiedColor=<rgba 1 1 1 1>}
     uVRegion {bIsValid=false, max=<0 0>, min=<0 0>}
    }
    renderTransform {angle=0, scale=<1 1.25>, shear=<0 0>, translation=<0 0>}
   }
   slot {class=OverlaySlot, properties {horizontalAlignment=HAlign_Fill, verticalAlignment=VAlign_Fill, padding=<ltrb -192 -96 -192 -32>}}
  }
  HorizontalBox_5 {widgetClassPath=&HorizontalBox, parent=Overlay_2, widget=HorizontalBox_5, slot {class=OverlaySlot}}
  Spacer_0 {widgetClassPath=&Spacer, parent=HorizontalBox_5, widget=Spacer_0, properties {size=<32 1>}, slot {class=HorizontalBoxSlot}}
  SizeBox_2 {
   widgetClassPath=&SizeBox
   parent=HorizontalBox_5
   widget=SizeBox_2
   properties {bOverride_MaxDesiredWidth=true, bOverride_MinDesiredWidth=true, maxDesiredWidth=768, minDesiredWidth=384}
   slot {class=HorizontalBoxSlot, properties {verticalAlignment=VAlign_Center}}
  }
  VerticalBox_3 {
   widgetClassPath=&VerticalBox
   parent=SizeBox_2
   widget=VerticalBox_3
   slot {class=SizeBoxSlot, properties {horizontalAlignment=HAlign_Left}}
  }
  BattlepassUpgraded {
   widgetClassPath=&CommonTextBlock
   bIsVariable=true
   parent=VerticalBox_3
   widget=BattlepassUpgraded
   properties {
    justification=Center
    scrollStyle=&TextScrollStyle-Base_C
    style=/FortUILibrary/Styles/Typography/Foundation/TextStyle-Header-XL-Purple.TextStyle-Header-XL-Purple_C
    text=FORTNITEMARES
    colorAndOpacity {colorUseRule=UseColor_Specified, specifiedColor=<rgba 0.3005 0.0052 0.5583 1>}
    font {
     bForceMonospaced=false
     bMaterialIsStencil=false
     fontMaterial=null
     fontObject=&Burbank
     letterSpacing=0
     monospacedWidth=1
     size=96
     skewAmount=0
     typefaceFontName=Burbank-Black
     outlineSettings {
      bApplyOutlineToDropShadows=false
      bMiteredCorners=false
      bSeparateFillAlpha=false
      outlineMaterial=null
      outlineSize=0
      outlineColor=<rgba 0 0 0 1>
     }
    }
    shadowColorAndOpacity=<rgba 0.9131 0.0103 0.7758 1>
    shadowOffset=<0 4>
   }
   slot {class=VerticalBoxSlot, properties {horizontalAlignment=HAlign_Center, padding=<ltrb 0 0 0 -16>}}
  }
  SubHeader {
   widgetClassPath=&CommonTextBlock
   bIsVariable=true
   parent=VerticalBox_3
   widget=SubHeader
   properties {
    justification=Center
    text="THANKS FOR PLAYING!"
    wrapTextAt=512
    colorAndOpacity {colorUseRule=UseColor_Specified, specifiedColor=<rgba 0.9131 0.0103 0.7758 1>}
    font {
     bForceMonospaced=false
     bMaterialIsStencil=false
     fontMaterial=null
     fontObject=&BurbankSmall
     letterSpacing=0
     monospacedWidth=1
     size=32
     skewAmount=0
     typefaceFontName=Black
     outlineSettings {
      bApplyOutlineToDropShadows=false
      bMiteredCorners=false
      bSeparateFillAlpha=false
      outlineMaterial=null
      outlineSize=0
      outlineColor=<rgba 0 0 0 1>
     }
    }
   }
   slot {class=VerticalBoxSlot, properties {horizontalAlignment=HAlign_Center}}
  }
  Glow {
   widgetClassPath=&Image
   bIsVariable=true
   parent=Overlay_1
   widget=Glow
   properties {
    visibility=SelfHitTestInvisible
    brush {
     bIsDynamicallyLoaded=false
     drawAs=Image
     imageType=NoImage
     mirroring=NoMirror
     resourceName=None
     resourceObject=/FortUILibrary/Styles/Border/Materials/M_Border_SoftGlow
     tiling=NoTile
     imageSize=<768 176>
     margin=<ltrb 0 0 0 0>
     outlineSettings {
      bUseBrushTransparency=false
      roundingType=HalfHeightRadius
      width=0
      color {colorUseRule=UseColor_Specified, specifiedColor=<rgba 0 0 0 0>}
      cornerRadii=<0 0 0 0>
     }
     tintColor {colorUseRule=UseColor_Specified, specifiedColor=<rgba 1 0 0.8681 1>}
     uVRegion {bIsValid=false, max=<0 0>, min=<0 0>}
    }
    colorAndOpacity=<rgba 1 1 1 0.4>
   }
   slot {class=OverlaySlot, properties {horizontalAlignment=HAlign_Fill, verticalAlignment=VAlign_Center, padding=<ltrb -128 0 -128 -48>}}
  }
 ]
}
