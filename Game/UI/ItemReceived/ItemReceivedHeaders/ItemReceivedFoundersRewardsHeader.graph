WidgetBlueprint /Game/UI/ItemReceived/ItemReceivedHeaders/ItemReceivedFoundersRewardsHeader
graphs [EventGraph {nodes [
   "SetText (Text)"(
    self=Get TopText() -> (TopText:Common Text Object Reference)
    InText="To Text (String)"(InString=Get FromUserName_BP() -> (FromUserName_BP:String))
   )
   > "SetText (Text)"(
    self=Get CommonTextBlock_0() -> (CommonTextBlock_0:Common Text Object Reference)
    InText=FormatTextUsingParams!(
     self:Fort Gift Box Item Object Reference=Get GiftBoxItem_BP() -> (GiftBoxItem_BP:Fort Gift Box Item Object Reference)
     FormatString:"Text (by ref)"=MakeLiteralText(Value="LOCTEXT(\"{userMessage}\")")
    ) -> (ReturnValue:Text)
   )
   Event InitFromGiftBoxItem_BP() {is=title K2Node_Event}
   > Parent: InitFromGiftBoxItem_BP() {is=title K2Node_CallParentFunction}
  ]}]
variables []
widgets {
 info {
  inheritedWidgetCount=0
  parentClass=&/ItemReceivedHeaderSubWidgetBase.ItemReceivedHeaderSubWidgetBase_C
  rootWidgetClass=&VerticalBox
  widgetCount=9
 }
 widgets [
  VerticalBox_8 {widgetClassPath=&VerticalBox, widget=VerticalBox_8}
  BorderItemNAme {
   widgetClassPath=&Border
   parent=VerticalBox_8
   widget=BorderItemNAme
   properties {
    background {
     bIsDynamicallyLoaded=false
     drawAs=NoDrawType
     imageType=NoImage
     mirroring=NoMirror
     resourceName=None
     resourceObject=null
     tiling=NoTile
     imageSize=<32 32>
     margin=<ltrb 0 0 0 0>
     outlineSettings {
      bUseBrushTransparency=false
      roundingType=HalfHeightRadius
      width=0
      color {colorUseRule=UseColor_Specified, specifiedColor=<rgba 0 0 0 0>}
      cornerRadii=<0 0 0 0>
     }
     tintColor {colorUseRule=UseColor_Specified, specifiedColor=<rgba 1 1 1 1>}
     uVRegion {bIsValid=false, max=<0 0>, min=<0 0>}
    }
    padding=<ltrb 0 0 0 0>
   }
   slot {class=VerticalBoxSlot, properties {horizontalAlignment=HAlign_Center, verticalAlignment=VAlign_Center}}
  }
  Overlay_9 {widgetClassPath=&Overlay, parent=BorderItemNAme, widget=Overlay_9, slot {class=BorderSlot, properties {padding=<ltrb 0 0 0 0>}}}
  BorderRarityGlow {
   widgetClassPath=&Border
   parent=Overlay_9
   widget=BorderRarityGlow
   properties {
    background {
     bIsDynamicallyLoaded=false
     drawAs=NoDrawType
     imageType=NoImage
     mirroring=NoMirror
     resourceName=None
     resourceObject=null
     tiling=NoTile
     imageSize=<32 32>
     margin=<ltrb 0 0 0 0>
     outlineSettings {
      bUseBrushTransparency=false
      roundingType=HalfHeightRadius
      width=0
      color {colorUseRule=UseColor_Specified, specifiedColor=<rgba 0 0 0 0>}
      cornerRadii=<0 0 0 0>
     }
     tintColor {colorUseRule=UseColor_Specified, specifiedColor=<rgba 1 1 1 1>}
     uVRegion {bIsValid=false, max=<0 0>, min=<0 0>}
    }
    padding=<ltrb 0 0 0 0>
   }
   slot {class=OverlaySlot, properties {horizontalAlignment=HAlign_Fill, verticalAlignment=VAlign_Fill, padding=<ltrb -64 0 -64 0>}}
  }
  TopText {
   widgetClassPath=&CommonTextBlock
   bIsVariable=true
   parent=Overlay_9
   widget=TopText
   properties {
    justification=Center
    scrollStyle=&TextScrollStyle-Base_C
    style=/FortUILibrary/Styles/Typography/Foundation/TextStyle-Header-XL-Horizon.TextStyle-Header-XL-Horizon_C
    text="Founder's Supply Drop!"
    font {
     bForceMonospaced=false
     bMaterialIsStencil=false
     fontMaterial=/FortUILibrary/Materials/Foundation/M_Font_Horizon_Dynamic
     fontObject=&Burbank
     letterSpacing=0
     monospacedWidth=1
     size=72
     skewAmount=0
     typefaceFontName=Burbank-Black
     outlineSettings {
      bApplyOutlineToDropShadows=true
      bMiteredCorners=false
      bSeparateFillAlpha=false
      outlineMaterial=null
      outlineSize=3
      outlineColor=<rgba 0 0 0 0.1>
     }
    }
    shadowOffset=<0 0>
   }
   slot {class=OverlaySlot}
  }
  HorizontalBox_3 {
   widgetClassPath=&HorizontalBox
   parent=VerticalBox_8
   widget=HorizontalBox_3
   slot {class=VerticalBoxSlot, properties {horizontalAlignment=HAlign_Center}}
  }
  BorderSuccessText {
   widgetClassPath=&Border
   parent=HorizontalBox_3
   widget=BorderSuccessText
   properties {
    background {
     bIsDynamicallyLoaded=false
     drawAs=NoDrawType
     imageType=NoImage
     mirroring=NoMirror
     resourceName=None
     resourceObject=null
     tiling=NoTile
     imageSize=<32 32>
     margin=<ltrb 0 0 0 0>
     outlineSettings {
      bUseBrushTransparency=false
      roundingType=HalfHeightRadius
      width=0
      color {colorUseRule=UseColor_Specified, specifiedColor=<rgba 0 0 0 0>}
      cornerRadii=<0 0 0 0>
     }
     tintColor {colorUseRule=UseColor_Specified, specifiedColor=<rgba 1 1 1 1>}
     uVRegion {bIsValid=false, max=<0 0>, min=<0 0>}
    }
    padding=<ltrb 0 0 0 0>
   }
   slot {class=HorizontalBoxSlot, properties {verticalAlignment=VAlign_Center}}
  }
  Subtext {
   widgetClassPath=&CommonTextBlock
   bIsVariable=true
   parent=BorderSuccessText
   widget=Subtext
   properties {
    justification=Center
    style=&TextStyle-Base_C
    text=Here's some cool stuff. Thanks for being a Fortnite Founder.
    font {
     bForceMonospaced=false
     bMaterialIsStencil=false
     fontMaterial=null
     fontObject=&NotoSans
     letterSpacing=0
     monospacedWidth=1
     size=24
     skewAmount=0
     typefaceFontName=NotoSans-Regular
     outlineSettings {
      bApplyOutlineToDropShadows=false
      bMiteredCorners=false
      bSeparateFillAlpha=false
      outlineMaterial=null
      outlineSize=0
      outlineColor=<rgba 0 0 0 1>
     }
    }
    shadowOffset=<0 0>
   }
   slot {class=BorderSlot, properties {padding=<ltrb 0 0 0 0>}}
  }
  CommonTextBlock_0 {
   widgetClassPath=&CommonTextBlock
   bIsVariable=true
   parent=VerticalBox_8
   widget=CommonTextBlock_0
   properties {
    justification=Center
    style=/FortUILibrary/Styles/Typography/Foundation/TextStyle-Base-M.TextStyle-Base-M_C
    font {
     bForceMonospaced=false
     bMaterialIsStencil=false
     fontMaterial=null
     fontObject=&NotoSans
     letterSpacing=0
     monospacedWidth=1
     size=24
     skewAmount=0
     typefaceFontName=NotoSans-Regular
     outlineSettings {
      bApplyOutlineToDropShadows=false
      bMiteredCorners=false
      bSeparateFillAlpha=false
      outlineMaterial=null
      outlineSize=0
      outlineColor=<rgba 0 0 0 1>
     }
    }
    shadowOffset=<0 0>
   }
   slot {class=VerticalBoxSlot, properties {padding=<ltrb 0 20 0 20>}}
  }
 ]
}
