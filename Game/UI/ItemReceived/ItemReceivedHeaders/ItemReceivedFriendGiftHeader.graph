WidgetBlueprint /Game/UI/ItemReceived/ItemReceivedHeaders/ItemReceivedFriendGiftHeader
graphs [EventGraph {nodes [
   n1 = CustomEvent_0!() -> (OutputDelegate, then, bSuccess:bool=~, SanitizedMsg:String) {is=title K2Node_CustomEvent}
   > "SetText (Text)"(
    self=Get CommonTextBlock_UserMessage() -> (CommonTextBlock_UserMessage:Common Text Object Reference)
    InText="To Text (String)"(InString=@n1.SanitizedMsg)
   )
   Event OnInitialized()
   > "SetText (Text)"(self=Get CommonTextBlock_UserMessage(), InText=GetEmptyText!() -> (ReturnValue:Text))
   Event InitFromGiftBoxItem_BP() {is=title K2Node_Event}
   > Parent: InitFromGiftBoxItem_BP() {is=title K2Node_CallParentFunction}
   > "SetText (Text)"(
    self=Get TextName() -> (TextName:Common Text Object Reference)
    InText="To Text (String)"(InString=Get FromUserName_BP() -> (FromUserName_BP:String))
   )
   > FilterUserMessageForProfanity!(
    execute
    self:Fort Gift Box Item Object Reference=Get GiftBoxItem_BP() -> (GiftBoxItem_BP:Fort Gift Box Item Object Reference)
    ControllerId:int=Get Local Player Controller ID!(Player:Player Controller Object Reference=GetOwningPlayer()) -> (ReturnValue:int=0) {is=title K2Node_CallFunction}
    OnTextFiltered:Delegate=@n1
   ) -> (then)
   n2 = Get FromUserName_BP()
   > "SetText (Text)"(
    self=Get CommonTextBlock_1() -> (CommonTextBlock_1:Common Text Object Reference)
    InText=Format Text!(Format="LOCTEXT(\"-{name}\")", name:Text="To Text (String)"(InString=@n2)) -> (Result)
   )
  ]}]
variables []
widgets {
 info {
  inheritedWidgetCount=0
  parentClass=&/ItemReceivedHeaderSubWidgetBase.ItemReceivedHeaderSubWidgetBase_C
  rootWidgetClass=&VerticalBox
  widgetCount=10
 }
 widgets [
  VerticalBox_3 {widgetClassPath=&VerticalBox, widget=VerticalBox_3}
  HorizontalBox_1 {
   widgetClassPath=&HorizontalBox
   parent=VerticalBox_3
   widget=HorizontalBox_1
   slot {class=VerticalBoxSlot, properties {horizontalAlignment=HAlign_Center}}
  }
  SizeBox_0 {
   widgetClassPath=&SizeBox
   parent=HorizontalBox_1
   widget=SizeBox_0
   properties {bOverride_MaxDesiredWidth=true, maxDesiredWidth=600}
   slot {class=HorizontalBoxSlot}
  }
  BorderItemNAme {
   widgetClassPath=&Border
   parent=SizeBox_0
   widget=BorderItemNAme
   properties {
    horizontalAlignment=HAlign_Center
    verticalAlignment=VAlign_Center
    background {
     bIsDynamicallyLoaded=false
     drawAs=NoDrawType
     imageType=NoImage
     mirroring=NoMirror
     resourceName=None
     resourceObject=null
     tiling=NoTile
     imageSize=<32 32>
     margin=<ltrb 0 0 0 0>
     outlineSettings {
      bUseBrushTransparency=false
      roundingType=HalfHeightRadius
      width=0
      color {colorUseRule=UseColor_Specified, specifiedColor=<rgba 0 0 0 0>}
      cornerRadii=<0 0 0 0>
     }
     tintColor {colorUseRule=UseColor_Specified, specifiedColor=<rgba 1 1 1 1>}
     uVRegion {bIsValid=false, max=<0 0>, min=<0 0>}
    }
    padding=<ltrb 0 4 0 0>
   }
   slot {class=SizeBoxSlot}
  }
  TextName {
   widgetClassPath=&CommonTextBlock
   bIsVariable=true
   parent=BorderItemNAme
   widget=TextName
   properties {
    justification=Center
    scrollStyle=&TextScrollStyle-Base_C
    style=/FortUILibrary/Styles/Typography/Foundation/TextStyle-Header-L.TextStyle-Header-L_C
    text="[PH]Gift Sender"
    font {
     bForceMonospaced=false
     bMaterialIsStencil=false
     fontMaterial=null
     fontObject=&Burbank
     letterSpacing=0
     monospacedWidth=1
     size=48
     skewAmount=0
     typefaceFontName=Burbank-Black
     outlineSettings {
      bApplyOutlineToDropShadows=false
      bMiteredCorners=false
      bSeparateFillAlpha=false
      outlineMaterial=null
      outlineSize=0
      outlineColor=<rgba 0 0 0 1>
     }
    }
    shadowOffset=<0 0>
   }
   slot {class=BorderSlot, properties {horizontalAlignment=HAlign_Center, verticalAlignment=VAlign_Center, padding=<ltrb 0 4 0 0>}}
  }
  BorderSuccessText {
   widgetClassPath=&Border
   parent=HorizontalBox_1
   widget=BorderSuccessText
   properties {
    horizontalAlignment=HAlign_Left
    verticalAlignment=VAlign_Center
    background {
     bIsDynamicallyLoaded=false
     drawAs=NoDrawType
     imageType=NoImage
     mirroring=NoMirror
     resourceName=None
     resourceObject=null
     tiling=NoTile
     imageSize=<32 32>
     margin=<ltrb 0 0 0 0>
     outlineSettings {
      bUseBrushTransparency=false
      roundingType=HalfHeightRadius
      width=0
      color {colorUseRule=UseColor_Specified, specifiedColor=<rgba 0 0 0 0>}
      cornerRadii=<0 0 0 0>
     }
     tintColor {colorUseRule=UseColor_Specified, specifiedColor=<rgba 1 1 1 1>}
     uVRegion {bIsValid=false, max=<0 0>, min=<0 0>}
    }
    padding=<ltrb 20 9 0 0>
   }
   slot {class=HorizontalBoxSlot}
  }
  Sender {
   widgetClassPath=&CommonTextBlock
   bIsVariable=true
   parent=BorderSuccessText
   widget=Sender
   properties {
    justification=Center
    text="SENT YOU A GIFT!"
    font {
     bForceMonospaced=false
     bMaterialIsStencil=false
     fontMaterial=null
     fontObject=&BurbankSmall
     letterSpacing=0
     monospacedWidth=1
     size=32
     skewAmount=0
     typefaceFontName=NotoSans-Regular
     outlineSettings {
      bApplyOutlineToDropShadows=false
      bMiteredCorners=false
      bSeparateFillAlpha=false
      outlineMaterial=null
      outlineSize=0
      outlineColor=<rgba 0 0 0 1>
     }
    }
   }
   slot {class=BorderSlot, properties {horizontalAlignment=HAlign_Left, verticalAlignment=VAlign_Center, padding=<ltrb 20 9 0 0>}}
  }
  VerticalBox_1 {widgetClassPath=&VerticalBox, parent=VerticalBox_3, widget=VerticalBox_1, slot {class=VerticalBoxSlot}}
  CommonTextBlock_UserMessage {
   widgetClassPath=&CommonTextBlock
   bIsVariable=true
   parent=VerticalBox_1
   widget=CommonTextBlock_UserMessage
   properties {
    justification=Center
    text="[ph] Lorem ipsum dolor sit amet, consectetur adipiscing elit. Nam dictum nisl dui, a consequat mi mattis convallis. Aenean eleifend porttitor risus, imperdiet dictum risus porta at. Maecenas "
    wrapTextAt=850
    wrappingPolicy=AllowPerCharacterWrapping
    font {
     bForceMonospaced=false
     bMaterialIsStencil=false
     fontMaterial=null
     fontObject=&BurbankSmall
     letterSpacing=0
     monospacedWidth=1
     size=24
     skewAmount=0
     typefaceFontName=NotoSans-Regular
     outlineSettings {
      bApplyOutlineToDropShadows=false
      bMiteredCorners=false
      bSeparateFillAlpha=false
      outlineMaterial=null
      outlineSize=0
      outlineColor=<rgba 0 0 0 1>
     }
    }
   }
   slot {class=VerticalBoxSlot, properties {verticalAlignment=VAlign_Bottom, padding=<ltrb 0 20 0 0>}}
  }
  CommonTextBlock_1 {
   widgetClassPath=&CommonTextBlock
   bIsVariable=true
   parent=VerticalBox_1
   widget=CommonTextBlock_1
   properties {
    justification=Center
    text="[PH] -UserName"
    visibility=Collapsed
    wrapTextAt=850
    font {
     bForceMonospaced=false
     bMaterialIsStencil=false
     fontMaterial=null
     fontObject=&BurbankSmall
     letterSpacing=0
     monospacedWidth=1
     size=24
     skewAmount=0
     typefaceFontName=Black
     outlineSettings {
      bApplyOutlineToDropShadows=false
      bMiteredCorners=false
      bSeparateFillAlpha=false
      outlineMaterial=null
      outlineSize=0
      outlineColor=<rgba 0 0 0 1>
     }
    }
   }
   slot {class=VerticalBoxSlot, properties {horizontalAlignment=HAlign_Center}}
  }
 ]
}
