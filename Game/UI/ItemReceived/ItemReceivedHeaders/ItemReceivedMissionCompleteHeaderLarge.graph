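# Large item-received header widget that announces a completed mission objective, showing quest name and challenge set name.
# On InitFromGiftBoxItem_BP, UpdateText2 resolves bundle and quest definitions from gift box parameters, then applies the bundle's display-style colors and background through dynamic materials.
# Gameplay tags and the quest's bundle level select Limited-Time, BattlePass or Prestige label and icon, collapsing both for some non-mission bundles.
# The older UpdateText variant is not called here.
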WidgetBlueprint /Game/UI/ItemReceived/ItemReceivedHeaders/ItemReceivedMissionCompleteHeaderLarge
graphs [
 EventGraph {nodes [
   Event InitFromGiftBoxItem_BP() {is=title K2Node_Event}
   > Parent: InitFromGiftBoxItem_BP() {is=title K2Node_CallParentFunction}
   > UpdateText2()
  ]}
 UpdateText {nodes [
   UpdateText() {is=title K2Node_FunctionEntry}
   n1 = Get GiftBoxItem_BP() -> (GiftBoxItem_BP:Fort Gift Box Item Object Reference)
   > n2 = Cast To FortChallengeBundleItemDefinition!(
    execute
    Object=RetrieveItemDefintionFromParam!(self:Fort Gift Box Item Object Reference=@n1, ParamName:String=BundleTemplateId) -> (ReturnValue:Fort Item Definition Object Reference)
   ) -> (then, CastFailed, AsFort Challenge Bundle Item Definition:Fort Challenge Bundle Item Definition Object Reference) {is=title K2Node_DynamicCast}
   > n3 = Cast To FortQuestItemDefinition!(execute, Object=RetrieveItemDefintionFromParam(self=@n1, ParamName=QuestTemplateId)) -> (then, CastFailed, AsFort Quest Item Definition:Fort Quest Item Definition Object Reference) {is=title K2Node_DynamicCast}
   n4 = Get ItemName!(self:Item Definition Base Object Reference=@n2) -> (ItemName:Text)
   > "SetText (Text)"(
    self=Get CommonTextBlock_2() -> (CommonTextBlock_2:Common Text Object Reference)
    InText=Format Text!(
     Format="LOCTEXT(\"You have received a reward for completing the {Quest Name} Objective Inside the {Mission Name} Mission!\")"
     Quest Name:Text=Get ItemName(self=@n3)
     Mission Name:Text=@n4
    ) -> (Result)
   )
   n5 = Get Text_ChallengeSetName() -> (Text_ChallengeSetName:Common Text Object Reference)
   > "SetText (Text)"(self=@n5, InText=@n4)
   n6 = Break Fort Challenge Set Style!(FortChallengeSetStyle:"Fort Challenge Set Style Structure (by ref)"=GetDisplayStyle!(self:Fort Challenge Bundle Item Definition Object Reference=@n2) -> (ReturnValue:"Fort Challenge Set Style Structure (by ref)")) -> (
    PrimaryColor:Linear Color Structure
    SecondaryColor:Linear Color Structure
    AccentColor:Linear Color Structure
    Context_LimitedTimeColor:Linear Color Structure
    Context_BaseColor:Linear Color Structure
    DisplayImage:Soft Object Reference
    CustomBackground:Soft Object Reference
   ) {is=title K2Node_BreakStruct}
   > SetColorAndOpacity(self=@n5, InColorAndOpacity=Make Slate Color(SpecifiedColor=@n6.AccentColor))
   n7 = Get LazyImage_Background() -> (LazyImage_Background:Fort Lazy Image Object Reference)
   > SetMaterialTextureParamName!(execute, self:Common Lazy Image Object Reference=@n7, TextureParamName:Name=CustomBackground) -> (then)
   > SetBrushFromLazyDisplayAsset!(
    execute
    self:Common Lazy Image Object Reference=@n7
    LazyObject:"Soft Object Reference (by ref)"=@n6.CustomBackground
    bMatchTextureSize:bool=false
   ) -> (then)
   > n8 = GetDynamicMaterial(self=@n7)
   > SetVectorParameterValue(self=@n8, ParameterName=Primary, Value=@n6)
   > SetVectorParameterValue(self=@n8, ParameterName=Secondary, Value=@n6.SecondaryColor)
   > SetVectorParameterValue(self=@n8, ParameterName=Accent, Value=@n6.AccentColor)
   n9 = Get Owned Gameplay Tags!(self:Gameplay Tag Asset Interface Interface=@n2) -> (ReturnValue:Gameplay Tag Container Structure) {is=title K2Node_CallFunction}
   > n10 = Branch(Condition=HasTag!(
     TagContainer:"Gameplay Tag Container Structure (by ref)"=@n9
     Tag:Gameplay Tag Structure=(TagName="ChallengeCategory.Mission.Urgent")
     bExactMatch:bool=true
    ) -> (ReturnValue:bool))
   n11 = Get Text_ChallengeContext() -> (Text_ChallengeContext:Common Text Object Reference)
   > "SetText (Text)"(self=@n11, InText="LOCTEXT(\"Limited-Time\")")
   > n12 = GetDynamicFontMaterial!(execute, self:Text Object Reference=@n11) -> (then, ReturnValue:Material Instance Dynamic Object Reference)
   > SetVectorParameterValue(self=@n12, ParameterName=BaseColor, Value=@n6.Context_LimitedTimeColor)
   > SetScalarParameterValue(self=@n12, ParameterName=PrestigeActive, Value=0)
   > SetActiveWidgetIndex(self=Get Switcher_MissionIcon() -> (Switcher_MissionIcon:Widget Switcher Object Reference), Index=0)
   n13 = Branch(
    execute=@n10.else
    Condition="Equal (Integer)"(
     A=GetBundleLevelForQuest!(self:Fort Challenge Bundle Item Definition Object Reference=@n2, QuestDef:Fort Quest Item Definition Object Reference=@n3) -> (ReturnValue:int=0)
     B=0
    ) {is=title K2Node_CallFunction}
   )
   n14 = Get Text_ChallengeContext()
   > "SetText (Text)"(self=@n14, InText="LOCTEXT(\"BattlePass\")")
   > n15 = GetDynamicFontMaterial(self=@n11)
   > SetVectorParameterValue(self=@n15, ParameterName=BaseColor, Value=@n6.Context_BaseColor)
   > SetScalarParameterValue(self=@n15, ParameterName=PrestigeActive, Value=0)
   n16 = HasTag(TagContainer=@n9, Tag=(TagName="ChallengeCategory.Mission"), bExactMatch=true)
   > n17 = Branch(Condition=@n16)
   n18 = Get Switcher_MissionIcon()
   > SetActiveWidgetIndex(self=@n18, Index=1)
   SetVisibility(execute=@n17.else, self=[@n18, Get Switcher_MissionIcon()], InVisibility=Hidden)
   "SetText (Text)"(execute=@n13.else, self=@n14, InText="LOCTEXT(\"Prestige\")")
   > n19 = GetDynamicFontMaterial(self=@n11)
   > SetVectorParameterValue(self=@n19, ParameterName=BaseColor, Value=@n6.Context_BaseColor)
   > SetScalarParameterValue(self=@n19, ParameterName=PrestigeActive, Value=1)
   > n20 = Branch(Condition=@n16)
   n21 = Get Switcher_MissionIcon()
   > SetActiveWidgetIndex(self=@n21, Index=2)
   SetVisibility(execute=@n20.else, self=@n21, InVisibility=Hidden)
   Get Hex Code for Style Color!(self:Fort Challenge Bundle Item Definition Object Reference=@n2, DesiredColor:EBundleStyleColor Enum=Primary) -> (ReturnValue:String) {is=title K2Node_CallFunction}
  ]}
 UpdateText2 {nodes [
   UpdateText2() {is=title K2Node_FunctionEntry}
   n1 = Get GiftBoxItem_BP()
   > n2 = Cast To FortChallengeBundleItemDefinition(Object=RetrieveItemDefintionFromParam(self=@n1, ParamName=BundleTemplateId))
   > n3 = Cast To FortQuestItemDefinition(Object=RetrieveItemDefintionFromParam(self=@n1, ParamName=QuestTemplateId))
   > "SetText (Text)"(self=Get CommonTextBlock_2(), InText=Format Text!(Format="LOCTEXT(\"{Quest Name}\")", Quest Name=Get ItemName(self=@n3)) -> (Result))
   n4 = Get Text_ChallengeSetName()
   > "SetText (Text)"(self=@n4, InText=Get ItemName(self=@n2))
   n5 = GetDisplayStyle(self=@n2)
   n6 = Break Fort Challenge Set Style(FortChallengeSetStyle=@n5)
   > SetColorAndOpacity(self=@n4, InColorAndOpacity=Make Slate Color(SpecifiedColor=@n6.AccentColor))
   > ApplyChallengeSetStyleColorsToImageMaterial!(
    execute
    DisplayStyle:"Fort Challenge Set Style Structure (by ref)"=@n5
    ImageWidget:Image Object Reference=Get HeaderBase() -> (HeaderBase:Image Object Reference)
   ) -> (then)
   n7 = Get LazyImage_Background()
   > SetMaterialTextureParamName(self=@n7, TextureParamName=CustomBackground)
   > SetBrushFromLazyDisplayAsset(self=@n7, LazyObject=@n6.CustomBackground, bMatchTextureSize=false)
   > n8 = GetDynamicMaterial(self=@n7)
   > SetVectorParameterValue(self=@n8, ParameterName=Primary, Value=@n6)
   > SetVectorParameterValue(self=@n8, ParameterName=Secondary, Value=@n6.SecondaryColor)
   > SetVectorParameterValue(self=@n8, ParameterName=Accent, Value=@n6.AccentColor)
   n9 = Get Owned Gameplay Tags(self=@n2)
   > n10 = Branch(Condition=HasTag(TagContainer=@n9, Tag=(TagName="ChallengeCategory.Mission.Urgent"), bExactMatch=true))
   n11 = Get Text_ChallengeContext()
   > "SetText (Text)"(self=@n11, InText="LOCTEXT(\"Limited-Time\")")
   > n12 = GetDynamicFontMaterial(self=@n11)
   > SetVectorParameterValue(self=@n12, ParameterName=BaseColor, Value=@n6.Context_LimitedTimeColor)
   > SetScalarParameterValue(self=@n12, ParameterName=PrestigeActive, Value=0)
   > SetActiveWidgetIndex(self=Get Switcher_MissionIcon(), Index=0)
   n13 = Branch(execute=@n10.else, Condition="Equal (Integer)"(A=GetBundleLevelForQuest(self=@n2, QuestDef=@n3) -> (ReturnValue=0), B=0))
   n14 = Get Text_ChallengeContext()
   > "SetText (Text)"(self=@n14, InText="LOCTEXT(\"BattlePass\")")
   > n15 = GetDynamicFontMaterial(self=@n11)
   > SetVectorParameterValue(self=@n15, ParameterName=BaseColor, Value=@n6.Context_BaseColor)
   > SetScalarParameterValue(self=@n15, ParameterName=PrestigeActive, Value=0)
   > SetActiveWidgetIndex(self=Get Switcher_MissionIcon(), Index=1)
   n16 = HasTag(TagContainer=@n9, Tag=(TagName="ChallengeCategory.Mission"), bExactMatch=false)
   > n17 = Branch(Condition=@n16)
   > SetActiveWidgetIndex(self=Get Switcher_MissionIcon(), Index=1)
   SetVisibility(execute=@n17.else, self=[Get Switcher_MissionIcon(), Get Text_ChallengeContext()], InVisibility=Collapsed)
   "SetText (Text)"(execute=@n13.else, self=@n14, InText="LOCTEXT(\"Prestige\")")
   > n18 = GetDynamicFontMaterial(self=@n11)
   > SetVectorParameterValue(self=@n18, ParameterName=BaseColor, Value=@n6.Context_BaseColor)
   > SetScalarParameterValue(self=@n18, ParameterName=PrestigeActive, Value=1)
   > SetActiveWidgetIndex(self=Get Switcher_MissionIcon(), Index=2)
   > n19 = Branch(Condition=@n16)
   > SetActiveWidgetIndex(self=Get Switcher_MissionIcon(), Index=2)
   SetVisibility(execute=@n19.else, self=[Get Text_ChallengeContext(), Get Switcher_MissionIcon()], InVisibility=Collapsed)
  ]}
]
variables []
widgets {
 info {
  inheritedWidgetCount=0
  parentClass=&/ItemReceivedHeaderSubWidgetBase.ItemReceivedHeaderSubWidgetBase_C
  rootWidgetClass=&Overlay
  widgetCount=21
 }
 widgets [
  Overlay_3 {widgetClassPath=&Overlay, widget=Overlay_3}
  SizeBox_1 {
   widgetClassPath=&SizeBox
   parent=Overlay_3
   widget=SizeBox_1
   properties {bOverride_WidthOverride=true, widthOverride=1065}
   slot {class=OverlaySlot, properties {horizontalAlignment=HAlign_Center, verticalAlignment=VAlign_Fill}}
  }
  Overlay_1 {widgetClassPath=&Overlay, parent=SizeBox_1, widget=Overlay_1, slot {class=SizeBoxSlot}}
  VerticalBox_3 {
   widgetClassPath=&VerticalBox
   parent=Overlay_1
   widget=VerticalBox_3
   slot {class=OverlaySlot, properties {horizontalAlignment=HAlign_Center, verticalAlignment=VAlign_Fill}}
  }
  SizeBox_0 {
   widgetClassPath=&SizeBox
   parent=VerticalBox_3
   widget=SizeBox_0
   properties {bOverride_MinDesiredHeight=true, bOverride_WidthOverride=true, minDesiredHeight=212, widthOverride=1065}
   slot {class=VerticalBoxSlot, properties {horizontalAlignment=HAlign_Center}}
  }
  MissionHeaderSB {
   widgetClassPath=&SizeBox
   parent=SizeBox_0
   widget=MissionHeaderSB
   properties {bOverride_HeightOverride=true, bOverride_WidthOverride=true, heightOverride=144, widthOverride=576}
   slot {class=SizeBoxSlot}
  }
  Overlay {widgetClassPath=&Overlay, parent=MissionHeaderSB, widget=Overlay, slot {class=SizeBoxSlot}}
  HeaderBase {
   widgetClassPath=&Image
   bIsVariable=true
   parent=Overlay
   widget=HeaderBase
   properties {
    brush {
     bIsDynamicallyLoaded=false
     drawAs=Image
     imageType=NoImage
     mirroring=NoMirror
     resourceName=None
     resourceObject=/Game/Athena/UI/Challenges/Art/MissionTileImages/M_UI_MissionHeaderBase
     tiling=NoTile
     imageSize=<32 32>
     margin=<ltrb 0 0 0 0>
     outlineSettings {
      bUseBrushTransparency=false
      roundingType=HalfHeightRadius
      width=0
      color {colorUseRule=UseColor_Specified, specifiedColor=<rgba 0 0 0 0>}
      cornerRadii=<0 0 0 0>
     }
     tintColor {colorUseRule=UseColor_Specified, specifiedColor=<rgba 1 1 1 1>}
     uVRegion {bIsValid=false, max=<0 0>, min=<0 0>}
    }
    renderTransformPivot=<0.5 0>
   }
   slot {class=OverlaySlot, properties {horizontalAlignment=HAlign_Fill, verticalAlignment=VAlign_Fill}}
  }
  LazyImage_Background {
   widgetClassPath=&FortLazyImage
   bIsVariable=true
   parent=Overlay
   widget=LazyImage_Background
   properties {
    brush {
     bIsDynamicallyLoaded=false
     drawAs=Image
     imageType=NoImage
     mirroring=NoMirror
     resourceName=None
     resourceObject=/Game/Athena/UI/Challenges/Art/MissionTileImages/M_UI_MissionHeaderMaster_Inst
     tiling=NoTile
     imageSize=<32 32>
     margin=<ltrb 0.5 0 0.5 0>
     outlineSettings {
      bUseBrushTransparency=false
      roundingType=HalfHeightRadius
      width=0
      color {colorUseRule=UseColor_Specified, specifiedColor=<rgba 0 0 0 0>}
      cornerRadii=<0 0 0 0>
     }
     tintColor {colorUseRule=UseColor_Specified, specifiedColor=<rgba 1 1 1 1>}
     uVRegion {bIsValid=false, max=<0 0>, min=<0 0>}
    }
    renderTransformPivot=<0.5 0>
   }
   slot {class=OverlaySlot, properties {horizontalAlignment=HAlign_Fill, verticalAlignment=VAlign_Fill}}
  }
  VerticalBox_1 {
   widgetClassPath=&VerticalBox
   parent=Overlay
   widget=VerticalBox_1
   slot {class=OverlaySlot, properties {horizontalAlignment=HAlign_Center, verticalAlignment=VAlign_Center, padding=<ltrb 0 0 0 28>}}
  }
  Switcher_MissionIcon {
   widgetClassPath=&WidgetSwitcher
   bIsVariable=true
   parent=VerticalBox_1
   widget=Switcher_MissionIcon
   slot {class=VerticalBoxSlot, properties {horizontalAlignment=HAlign_Center, verticalAlignment=VAlign_Top, padding=<ltrb 0 24 0 0>}}
  }
  Icon_LimitedTime {
   widgetClassPath=&Image
   bIsVariable=true
   parent=Switcher_MissionIcon
   widget=Icon_LimitedTime
   properties {
    visibility=SelfHitTestInvisible
    brush {
     bIsDynamicallyLoaded=false
     drawAs=Image
     imageType=NoImage
     mirroring=NoMirror
     resourceName=None
     resourceObject=/Game/Athena/UI/Challenges/Art/MissionTileImages/M_UI_Icon_Mission_LimitedTime_128
     tiling=NoTile
     imageSize=<64 64>
     margin=<ltrb 0 0 0 0>
     outlineSettings {
      bUseBrushTransparency=false
      roundingType=HalfHeightRadius
      width=0
      color {colorUseRule=UseColor_Specified, specifiedColor=<rgba 0 0 0 0>}
      cornerRadii=<0 0 0 0>
     }
     tintColor {colorUseRule=UseColor_Specified, specifiedColor=<rgba 1 1 1 1>}
     uVRegion {bIsValid=false, max=<0 0>, min=<0 0>}
    }
   }
   slot {class=WidgetSwitcherSlot}
  }
  Icon_BattlePass {
   widgetClassPath=&Image
   bIsVariable=true
   parent=Switcher_MissionIcon
   widget=Icon_BattlePass
   properties {
    visibility=SelfHitTestInvisible
    brush {
     bIsDynamicallyLoaded=false
     drawAs=Image
     imageType=NoImage
     mirroring=NoMirror
     resourceName=None
     resourceObject=/Game/Athena/UI/Challenges/Art/MissionTileImages/M_UI_Icon_Mission_BattlePass_128
     tiling=NoTile
     imageSize=<64 64>
     margin=<ltrb 0 0 0 0>
     outlineSettings {
      bUseBrushTransparency=false
      roundingType=HalfHeightRadius
      width=0
      color {colorUseRule=UseColor_Specified, specifiedColor=<rgba 0 0 0 0>}
      cornerRadii=<0 0 0 0>
     }
     tintColor {colorUseRule=UseColor_Specified, specifiedColor=<rgba 1 1 1 1>}
     uVRegion {bIsValid=false, max=<0 0>, min=<0 0>}
    }
   }
   slot {class=WidgetSwitcherSlot}
  }
  Icon_Prestige {
   widgetClassPath=&Image
   bIsVariable=true
   parent=Switcher_MissionIcon
   widget=Icon_Prestige
   properties {
    visibility=SelfHitTestInvisible
    brush {
     bIsDynamicallyLoaded=false
     drawAs=Image
     imageType=NoImage
     mirroring=NoMirror
     resourceName=None
     resourceObject=/Game/Athena/UI/Challenges/Art/MissionTileImages/M_UI_Icon_Mission_Prestige_128
     tiling=NoTile
     imageSize=<64 64>
     margin=<ltrb 0 0 0 0>
     outlineSettings {
      bUseBrushTransparency=false
      roundingType=HalfHeightRadius
      width=0
      color {colorUseRule=UseColor_Specified, specifiedColor=<rgba 0 0 0 0>}
      cornerRadii=<0 0 0 0>
     }
     tintColor {colorUseRule=UseColor_Specified, specifiedColor=<rgba 1 1 1 1>}
     uVRegion {bIsValid=false, max=<0 0>, min=<0 0>}
    }
   }
   slot {class=WidgetSwitcherSlot}
  }
  Text_ChallengeContext {
   widgetClassPath=&CommonTextBlock
   bIsVariable=true
   parent=VerticalBox_1
   widget=Text_ChallengeContext
   properties {
    justification=Center
    scrollStyle=/FortUILibrary/Styles/Typography/Foundation/TextScrollStyle-NoFade.TextScrollStyle-NoFade_C
    text=Limited Time
    textTransformPolicy=ToUpper
    font {
     bForceMonospaced=false
     bMaterialIsStencil=false
     fontMaterial=/FortUILibrary/Materials/Foundation/M_Font_Shine_Prestige
     fontObject=&BurbankSmall
     letterSpacing=0
     monospacedWidth=1
     size=18
     skewAmount=0
     typefaceFontName=Black
     outlineSettings {
      bApplyOutlineToDropShadows=false
      bMiteredCorners=false
      bSeparateFillAlpha=false
      outlineMaterial=null
      outlineSize=0
      outlineColor=<rgba 0 0 0 1>
     }
    }
    renderTransform {angle=0, scale=<1 1>, shear=<-12 0>, translation=<0 0>}
    shadowOffset=<0 0>
   }
   slot {class=VerticalBoxSlot, properties {verticalAlignment=VAlign_Center, padding=<ltrb 0 0 0 -4>}}
  }
  HeaderScaleBox {widgetClassPath=&ScaleBox, parent=VerticalBox_1, widget=HeaderScaleBox, slot {class=VerticalBoxSlot}}
  Text_ChallengeSetName {
   widgetClassPath=&CommonTextBlock
   bIsVariable=true
   parent=HeaderScaleBox
   widget=Text_ChallengeSetName
   properties {
    justification=Center
    text=Mission Name
    textTransformPolicy=ToUpper
    font {
     bForceMonospaced=false
     bMaterialIsStencil=false
     fontMaterial=null
     fontObject=&Burbank
     letterSpacing=0
     monospacedWidth=1
     size=40
     skewAmount=0
     typefaceFontName=BurbankRegular-Black
     outlineSettings {
      bApplyOutlineToDropShadows=false
      bMiteredCorners=false
      bSeparateFillAlpha=false
      outlineMaterial=null
      outlineSize=0
      outlineColor=<rgba 0 0 0 1>
     }
    }
    renderTransform {angle=0, scale=<1 1>, shear=<-12 0>, translation=<0 0>}
   }
   slot {class=ScaleBoxSlot, properties {horizontalAlignment=HAlign_Fill, verticalAlignment=VAlign_Fill}}
  }
  VerticalBox_2 {
   widgetClassPath=&VerticalBox
   parent=VerticalBox_3
   widget=VerticalBox_2
   slot {class=VerticalBoxSlot, properties {horizontalAlignment=HAlign_Center, verticalAlignment=VAlign_Top, padding=<ltrb 0 24 0 0>}}
  }
  BattlepassUpgraded {
   widgetClassPath=&CommonTextBlock
   bIsVariable=true
   parent=VerticalBox_2
   widget=BattlepassUpgraded
   properties {
    justification=Center
    text="OBJECTIVE COMPLETE!"
    textTransformPolicy=ToUpper
    font {
     bForceMonospaced=false
     bMaterialIsStencil=false
     fontMaterial=null
     fontObject=&BurbankSmall
     letterSpacing=0
     monospacedWidth=1
     size=24
     skewAmount=0
     typefaceFontName=Black
     outlineSettings {
      bApplyOutlineToDropShadows=false
      bMiteredCorners=false
      bSeparateFillAlpha=false
      outlineMaterial=null
      outlineSize=0
      outlineColor=<rgba 0 0 0 1>
     }
    }
    renderTransform {angle=0, scale=<1 1>, shear=<-12 0>, translation=<0 0>}
    shadowOffset=<0 0>
   }
   slot {class=VerticalBoxSlot}
  }
  CommonTextBlock_2 {
   widgetClassPath=&CommonTextBlock
   bIsVariable=true
   parent=VerticalBox_2
   widget=CommonTextBlock_2
   properties {
    justification=Center
    text="\"{Quest Name}\""
    wrapTextAt=512
    font {
     bForceMonospaced=false
     bMaterialIsStencil=false
     fontMaterial=null
     fontObject=&BurbankSmall
     letterSpacing=0
     monospacedWidth=1
     size=16
     skewAmount=0
     typefaceFontName=Bold
     outlineSettings {
      bApplyOutlineToDropShadows=false
      bMiteredCorners=false
      bSeparateFillAlpha=false
      outlineMaterial=null
      outlineSize=0
      outlineColor=<rgba 0 0 0 1>
     }
    }
    shadowOffset=<0 0>
   }
   slot {class=VerticalBoxSlot, properties {padding=<ltrb 0 12 0 0>, size {sizeRule=Fill, value=1}}}
  }
  ItemReceivedAccountWidget {
   widgetClassPath=&/ItemReceivedAccountWidget
   bIsVariable=true
   parent=VerticalBox_2
   widget=ItemReceivedAccountWidget
   properties {visibility=Collapsed}
   slot {class=VerticalBoxSlot}
  }
 ]
}
