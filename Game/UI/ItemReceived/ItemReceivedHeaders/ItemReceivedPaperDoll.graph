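# PaperDoll header sub-widget for the item-received UI, presenting a gift box item's image, header, subtitle and body.
# On init, it lazy-loads the definition's display image unless suppressed or unset, switching text between left and centered alignment accordingly.
# Header and body fall back from override param to keyed string-map lookup to definition default.
# Item params also fill text placeholders, supply the optional userMessage subtitle and toggle the Epic logo.
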WidgetBlueprint /Game/UI/ItemReceived/ItemReceivedHeaders/ItemReceivedPaperDoll
graphs [EventGraph {nodes [
   Event InitFromGiftBoxItem_BP() {is=title K2Node_Event}
   > Parent: InitFromGiftBoxItem_BP() {is=title K2Node_CallParentFunction}
   n1 = HasParam!(
    self:Fort Gift Box Item Object Reference=Get GiftBoxItem_BP() -> (GiftBoxItem_BP:Fort Gift Box Item Object Reference)
    ParamName:String=EpicLogo
   ) -> (ReturnValue:bool)
   n2 = MakeLiteralText(Value="LOCTEXT(\"{HeaderAssetString}\")")
   n3 = Get GiftBoxItem_BP()
   n4 = "Equal Exactly (String)"(
    A="To String (Text)"(InText=FormatTextUsingParams!(self:Fort Gift Box Item Object Reference=@n3, FormatString:"Text (by ref)"=@n2) -> (ReturnValue:Text))
    B=Header.EpicGames
   )
   > SetVisibility(
    self=Get SmallEpicLogo() -> (SmallEpicLogo:Image Object Reference)
    InVisibility=Select(Option 0=Collapsed, Option 1=SelfHitTestInvisible, Index=OR Boolean(A=@n4, B=@n1))
   )
   n5 = Get ItemDisplayAsset!(self:Fort Gift Box Item Definition Object Reference=GetGiftBoxDefinition!(self:Fort Gift Box Item Object Reference=Get GiftBoxItem_BP()) -> (ReturnValue:Fort Gift Box Item Definition Object Reference)) -> (ItemDisplayAsset:Soft Object Reference)
   n6 = Get Image() -> (Image:Common Lazy Image Object Reference)
   > SetBrushFromLazyDisplayAsset!(execute, self:Common Lazy Image Object Reference=@n6, LazyObject:"Soft Object Reference (by ref)"=@n5, bMatchTextureSize:bool=false) -> (then)
   n7 = TryGetParamString!(self:Fort Gift Box Item Object Reference=Get GiftBoxItem_BP(), ParamName:String=SuppressAssetImage) -> (OutValue:String, ReturnValue:bool)
   n8 = AND Boolean(A=IsValidSoftObjectReference(SoftObjectReference=@n5), B=NOT Boolean(A="Equal Exactly (String)"(A=@n7, B=true)))
   > SetVisibility(self=@n6, InVisibility=Select(Option 0=Collapsed, Option 1=SelfHitTestInvisible, Index=@n8))
   > SetJustification!(
    execute
    self:Text Layout Widget Object Reference=[
     Get ItemReceived_Subtitle() -> (ItemReceived_Subtitle:Common Text Object Reference)
     Get Header() -> (Header:Common Text Object Reference)
     Get Body() -> (Body:Common Text Object Reference)
    ]
    InJustification:ETextJustify Enum=Select(Option 0:ETextJustify Enum=Center, Option 1:ETextJustify Enum=Left, Index=@n8) -> (ReturnValue:ETextJustify Enum)
   ) -> (then)
   > n9 = Sequence!(execute) -> (then_0, then_1, then_2:Exec)
   n10 = Get GiftBoxItem_BP()
   n11 = TryGetParamString(self=@n10, ParamName=HeaderOverrideString)
   > n12 = Branch(Condition=@n11.ReturnValue)
   n13 = "To Text (String)"(InString=@n11)
   > "SetText (Text)"(self=Get Header(), InText=TextToUpper!(InText:"Text (by ref)"=FormatTextUsingParams(self=@n10, FormatString=@n13)) -> (ReturnValue:Text))
   n14 = Get GiftBoxItem_BP()
   n15 = TryGetParamString(self=@n14, ParamName=HeaderAssetString)
   n16 = Find(TargetMap:Map of Strings to Texts=Get AssetHeaderStrings() -> (AssetHeaderStrings:Map of Strings to Texts), Key=@n15) -> (Value:Text)
   n17 = Branch(execute=@n12.else, Condition=AND Boolean(A=@n15.ReturnValue, B=@n16.ReturnValue))
   > "SetText (Text)"(self=Get Header(), InText=TextToUpper(InText=FormatTextUsingParams(self=@n14, FormatString=@n16)))
   n18 = GetGiftBoxDefinition(self=Get GiftBoxItem_BP())
   "SetText (Text)"(
    execute=@n17.else
    self=Get Header()
    InText=TextToUpper(InText=Get DefaultHeaderText!(self:Fort Gift Box Item Definition Object Reference=@n18) -> (DefaultHeaderText:Text))
   )
   n19 = Get GiftBoxItem_BP()
   n20 = TryGetParamString(self=@n19, ParamName=BodyOverrideString)
   n21 = Branch(execute=@n9.then_1, Condition=@n20.ReturnValue)
   > "SetText (Text)"(self=Get Body(), InText=FormatTextUsingParams(self=@n19, FormatString="To Text (String)"(InString=@n20)))
   n22 = Get GiftBoxItem_BP()
   n23 = TryGetParamString(self=@n22, ParamName=BodyAssetString)
   n24 = Find(TargetMap=Get AssetBodyStrings() -> (AssetBodyStrings:Map of Strings to Texts), Key=@n23)
   n25 = Branch(execute=@n21.else, Condition=AND Boolean(A=@n23.ReturnValue, B=@n24.ReturnValue))
   > "SetText (Text)"(self=Get Body(), InText=FormatTextUsingParams(self=@n22, FormatString=@n24))
   n26 = Get GiftBoxItem_BP()
   "SetText (Text)"(
    execute=@n25.else
    self=Get Body()
    InText=Get DefaultBodyText!(self:Fort Gift Box Item Definition Object Reference=GetGiftBoxDefinition(self=@n26)) -> (DefaultBodyText:Text)
   )
   n27 = Get GiftBoxItem_BP()
   n28 = Branch(execute=@n9.then_2, Condition=NOT Boolean(A=HasParam(self=@n27, ParamName=userMessage)))
   n29 = Get ItemReceived_Subtitle()
   > SetVisibility(self=@n29, InVisibility=Collapsed)
   n30 = MakeLiteralText(Value="LOCTEXT(\"{userMessage}\")")
   "SetText (Text)"(execute=@n28.else, self=@n29, InText=TextToUpper(InText=FormatTextUsingParams(self=@n27, FormatString=@n30)))
  ]}]
variables [
 AssetBodyStrings {
  type="map<string>"
  default {
   AccountMerge.Complete="These items have been migrated from your old account. Thank you for your patience!"
   Body.PaperDoll="You've unlocked a new Upgrade! Head to the Upgrade Vault!"
   Compensation.Downtime="We apologize for the recent downtime, so we want you to have this. Good luck; have fun!"
   Compensation.GiftTokenRefund=You had a Battle Pass Gift Token from an attempt to gift last season's Battle Pass. It will now be converted into V-Bucks.
   Compensation.IncorrectBan=We incorrectly banned your account. We're sorry and will be granting V-Bucks to your account to compensate for the error.
   Compensation.IncorrectBan.VBucks="We incorrectly banned your account. We're sorry and will be granting 2,000 V-Bucks and the appropriate amount of Battle Pass Stars to your account to compensate for the error."
   Compensation.ItemOwed="This belongs to you. Thanks for your patience!"
   Competitive.TeamDisqualified=Your team has been disqualified from a recent tournament due to one or more team members being disqualified. Any rewards earned by this team have been removed.
   Inventory.Added=The following items were added to your account by customer support:
   Inventory.OutfitChanged="We've made some changes to this outfit. If you are no longer satisfied with your purchase, you can request a refund anytime before {date-Expiration} using the in-game flow (no refund token required). Thanks for playing!"
   Inventory.Removed=The following items were removed from your account by customer support:
   Migration.AddedItems="One of your cosmetic items now includes additional items. Enjoy!"
   Migration.AddedItems.BackBling="One of your cosmetic items now includes a matching Back Bling. Enjoy!"
   Migration.General="Some rewards are now given out differently, so you deserve this. Thanks for playing!"
   Migration.Removedtems=The following items were removed from your account by player support.
   Objective.LinkedAccount=You earned these rewards for inviting active friends:
   Reward.14DaysOfSummer="Thanks for playing during the 14 Days of Summer event!"
   Reward.14DaysOfSummerBundle="Welcome to the 14 Days of Summer event!"
   Reward.AffiliateWrap="You've unlocked the Cuddle Hearts Wrap!"
   Reward.AirRoyale="You've completed an Air Royale Challenge!  Enjoy your reward!"
   Reward.AshtonLTM="Thanks for playing Endgame! Enjoy!"
   Reward.Birthday2="Thanks for helping celebrate Fortnite's 2nd Birthday!"
   Reward.EgsFulfillment="One of your recent Epic Games purchases includes a special Fortnite reward, thank you! Drop in and have fun."
   Reward.EgsFulfillmentAlreadyOwned="One of your recent Epic Games purchases includes a special Fortnite reward that was already purchased, so we've returned those spent V-Bucks to you. Thank you!"
   Reward.FestivusEventStarted="Enjoy the show!"
   Reward.FortbyteLoadingScreen="100 Fortbytes Collected - Image Restored!"
   Reward.FortniteMares="Two spooky new heroes have joined your Homebase. Boo!"
   Reward.FortniteMares_CN="A spooky new hero has joined your Homebase. Boo!"
   Reward.GeForceBundle="Here's a variant of the Counterattack Set for purchasing the GeForce Fortnite Bundle!"
   Reward.Generic="There could be any number of reasons why you're getting this item. Enjoy!"
   Reward.GenericGift1="A gift from Epic. Enjoy!"
   Reward.GenericGift2="Loot! Enjoy this gift from Epic"
   Reward.GenericGift3="A new item for your collection. Enjoy!"
   Reward.GenericGift4="Well look who's got themselves some more goodness!"
   Reward.GenericGift5="Here is your reward. Enjoy!"
   Reward.LimitedTimeMashup="Thank you for playing the recent Limited Time Mashup. Enjoy!"
   Reward.LimitedTimeMode="Thank you for playing the recent Limited Time Mode. Enjoy!"
   Reward.Oak="Thanks for playing during the Welcome to Pandora event!"
   Reward.OvertimeEventStarted="Thanks for playing in Overtime! Enjoy!"
   Reward.Ruin="You've completed a Ruin Challenge!  Reap your reward!"
   Reward.Showdown.HighPlacement="Congratulations on your excellent performance in the Showdown LTM!"
   Reward.SoloShowdown.HighPlacement="Congratulations on your excellent performance in the Solo Showdown LTM!"
   Reward.SoloShowdownV1.Played50Games="Congratulations, you completed 50 games in the Solo Showdown LTM! As a thank you for participating, you have been awarded the unique Showdown Spray."
   Reward.SpecialEventStarted="A special event just started, complete these challenges before it ends!"
   Reward.StW.Special=This one's on us. Go forth and fortify.
   Reward.Tournament="Congratulations on your excellent performance in a recent tournament event!"
   Reward.VictoryRoyale="You earned a Victory Royale!"
  }
 }
 AssetHeaderStrings {
  type="map<string>"
  default {
   Header.AccountMerge="ACCOUNT MERGE COMPLETE!"
   Header.Congratulations="CONGRATULATIONS!"
   Header.EpicGames=EPIC GAMES
   Header.EpicGamesStore=EPIC GAMES STORE
   Header.GiftTokenRefund=BATTLE PASS GIFTING REFUND
   Header.IncorrectBan=INCORRECT BAN
   Header.ItemDrop="ITEM DROP!"
   Header.ItemReceived="ITEM RECEIVED!"
   Header.PaperDoll="MAYA CUSTOMIZATION UNLOCKED!"
   Header.PreReg=PRE-REGISTRATION EVENT
   Header.TeamDisqualified=TEAM DISQUALIFIED
  }
 }
]
widgets {
 info {
  inheritedWidgetCount=0
  parentClass=&/ItemReceivedHeaderSubWidgetBase.ItemReceivedHeaderSubWidgetBase_C
  rootWidgetClass=&HorizontalBox
  widgetCount=10
 }
 widgets [
  HorizontalBox_1 {widgetClassPath=&HorizontalBox, widget=HorizontalBox_1}
  SizeBox_0 {
   widgetClassPath=&SizeBox
   parent=HorizontalBox_1
   widget=SizeBox_0
   properties {bOverride_HeightOverride=true, bOverride_WidthOverride=true, heightOverride=256, widthOverride=256}
   slot {class=HorizontalBoxSlot, properties {horizontalAlignment=HAlign_Right, verticalAlignment=VAlign_Center}}
  }
  Image {
   widgetClassPath=/Script/CommonUI.CommonLazyImage
   bIsVariable=true
   parent=SizeBox_0
   widget=Image
   properties {
    visibility=HitTestInvisible
    brush {
     bIsDynamicallyLoaded=false
     drawAs=Image
     imageType=NoImage
     mirroring=NoMirror
     resourceName=None
     resourceObject=/Game/UI/Foundation/Textures/Icons/Items/T-FNBR-BattlePass-L
     tiling=NoTile
     imageSize=<512 512>
     margin=<ltrb 0 0 0 0>
     outlineSettings {
      bUseBrushTransparency=false
      roundingType=HalfHeightRadius
      width=0
      color {colorUseRule=UseColor_Specified, specifiedColor=<rgba 0 0 0 0>}
      cornerRadii=<0 0 0 0>
     }
     tintColor {colorUseRule=UseColor_Specified, specifiedColor=<rgba 1 1 1 1>}
     uVRegion {bIsValid=false, max=<0 0>, min=<0 0>}
    }
   }
   slot {class=SizeBoxSlot}
  }
  VerticalBox_1 {
   widgetClassPath=&VerticalBox
   parent=HorizontalBox_1
   widget=VerticalBox_1
   slot {class=HorizontalBoxSlot, properties {horizontalAlignment=HAlign_Left, verticalAlignment=VAlign_Center, padding=<ltrb 50 0 0 0>}}
  }
  VerticalBox_0 {
   widgetClassPath=&VerticalBox
   parent=VerticalBox_1
   widget=VerticalBox_0
   slot {class=VerticalBoxSlot, properties {horizontalAlignment=HAlign_Center}}
  }
  HorizontalBox_0 {
   widgetClassPath=&HorizontalBox
   parent=VerticalBox_1
   widget=HorizontalBox_0
   slot {class=VerticalBoxSlot, properties {horizontalAlignment=HAlign_Center, verticalAlignment=VAlign_Top, padding=<ltrb 0 -10 0 -10>}}
  }
  SmallEpicLogo {
   widgetClassPath=&Image
   bIsVariable=true
   parent=HorizontalBox_0
   widget=SmallEpicLogo
   properties {
    visibility=HitTestInvisible
    brush {
     bIsDynamicallyLoaded=false
     drawAs=Image
     imageType=NoImage
     mirroring=NoMirror
     resourceName=None
     resourceObject=/Game/UI/Foundation/Textures/Icons/Social/T-Icon-EpicGamesLogo-64
     tiling=NoTile
     imageSize=<64 64>
     margin=<ltrb 0 0 0 0>
     outlineSettings {
      bUseBrushTransparency=false
      roundingType=HalfHeightRadius
      width=0
      color {colorUseRule=UseColor_Specified, specifiedColor=<rgba 0 0 0 0>}
      cornerRadii=<0 0 0 0>
     }
     tintColor {colorUseRule=UseColor_Specified, specifiedColor=<rgba 1 1 1 1>}
     uVRegion {bIsValid=false, max=<0 0>, min=<0 0>}
    }
   }
   slot {class=HorizontalBoxSlot, properties {horizontalAlignment=HAlign_Right, verticalAlignment=VAlign_Center, padding=<ltrb -64 0 0 0>}}
  }
  Header {
   widgetClassPath=&CommonTextBlock
   bIsVariable=true
   parent=HorizontalBox_0
   widget=Header
   properties {
    justification=Center
    scrollStyle=&TextScrollStyle-Base_C
    text="{PaperDollHeader}"
    font {
     bForceMonospaced=false
     bMaterialIsStencil=false
     fontMaterial=null
     fontObject=&Burbank
     letterSpacing=0
     monospacedWidth=1
     size=72
     skewAmount=0
     typefaceFontName=Burbank-Black
     outlineSettings {
      bApplyOutlineToDropShadows=true
      bMiteredCorners=false
      bSeparateFillAlpha=false
      outlineMaterial=null
      outlineSize=3
      outlineColor=<rgba 0 0 0 0.1>
     }
    }
   }
   slot {class=HorizontalBoxSlot}
  }
  ItemReceived_Subtitle {
   widgetClassPath=&CommonTextBlock
   bIsVariable=true
   parent=VerticalBox_1
   widget=ItemReceived_Subtitle
   properties {
    justification=Center
    text="{PaperDollSubtitle}"
    wrapTextAt=800
    font {
     bForceMonospaced=false
     bMaterialIsStencil=false
     fontMaterial=null
     fontObject=&BurbankSmall
     letterSpacing=0
     monospacedWidth=1
     size=25
     skewAmount=0
     typefaceFontName=NotoSans-Regular
     outlineSettings {
      bApplyOutlineToDropShadows=false
      bMiteredCorners=false
      bSeparateFillAlpha=false
      outlineMaterial=null
      outlineSize=0
      outlineColor=<rgba 0 0 0 1>
     }
    }
   }
   slot {class=VerticalBoxSlot, properties {padding=<ltrb 0 0 0 15>}}
  }
  Body {
   widgetClassPath=&CommonTextBlock
   bIsVariable=true
   parent=VerticalBox_1
   widget=Body
   properties {
    justification=Center
    style=/FortUILibrary/Styles/Typography/Foundation/TextStyle-Base-M.TextStyle-Base-M_C
    text="{PaperDollBody}"
    wrapTextAt=768
    font {
     bForceMonospaced=false
     bMaterialIsStencil=false
     fontMaterial=null
     fontObject=&NotoSans
     letterSpacing=0
     monospacedWidth=1
     size=24
     skewAmount=0
     typefaceFontName=NotoSans-Regular
     outlineSettings {
      bApplyOutlineToDropShadows=false
      bMiteredCorners=false
      bSeparateFillAlpha=false
      outlineMaterial=null
      outlineSize=0
      outlineColor=<rgba 0 0 0 1>
     }
    }
    shadowOffset=<0 0>
   }
   slot {class=VerticalBoxSlot, properties {horizontalAlignment=HAlign_Center, padding=<ltrb 0 10 0 0>}}
  }
 ]
}
