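# Gift box header widget shown when Player Support restores items to an account, with Epic Games branding and an apology message.
# On init it calls the parent, then sets ReasonRichText to a transaction-reversed reason for self-gifts or gifts without an account ID, otherwise citing issues with the named gifter's account.
# A customReason parameter on the gift box item, when present, overrides that text.
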WidgetBlueprint /Game/UI/ItemReceived/ItemReceivedHeaders/ItemRestoredHeader
graphs [EventGraph {nodes [
   Parent: InitFromGiftBoxItem_BP() {is=title K2Node_CallParentFunction}
   n1 = NOT Boolean(A=HasAccountID!(self:Fort Gift Box Item Object Reference=Get GiftBoxItem_BP() -> (GiftBoxItem_BP:Fort Gift Box Item Object Reference)) -> (ReturnValue:bool))
   n2 = Get bSelfGift() -> (bSelfGift:bool)
   n3 = "To Text (String)"(InString=Get FromUserName_BP() -> (FromUserName_BP:String))
   > SetText(
    self=Get ReasonRichText() -> (ReasonRichText:Rich Text Block Object Reference)
    InText=Select(
     Option 0:Text=Format Text!(Format="LOCTEXT(\"Reason: Player Support identified issues with <PlayerName>{GiftersName}\\'s</> account.\")", GiftersName:Text=@n3) -> (Result)
     Option 1:Text="LOCTEXT(\"Reason: A transaction on this account was reversed.\")"
     Index=OR Boolean(A=@n1, B=@n2)
    ) -> (ReturnValue:Text)
   )
   n4 = Get GiftBoxItem_BP()
   > Branch(Condition=HasParam!(self:Fort Gift Box Item Object Reference=@n4, ParamName:String=customReason) -> (ReturnValue:bool))
   > SetText(
    self=Get ReasonRichText()
    InText="To Text (String)"(InString=TryGetParamString!(self:Fort Gift Box Item Object Reference=@n4, ParamName:String=customReason) -> (OutValue:String, ReturnValue:bool))
   )
   Event InitFromGiftBoxItem_BP() {is=title K2Node_Event}
  ]}]
variables []
widgets {
 info {
  inheritedWidgetCount=0
  parentClass=&/ItemReceivedHeaderSubWidgetBase.ItemReceivedHeaderSubWidgetBase_C
  rootWidgetClass=&VerticalBox
  widgetCount=11
 }
 widgets [
  VerticalBox_1 {widgetClassPath=&VerticalBox, widget=VerticalBox_1}
  VerticalBox_0 {
   widgetClassPath=&VerticalBox
   parent=VerticalBox_1
   widget=VerticalBox_0
   slot {class=VerticalBoxSlot, properties {horizontalAlignment=HAlign_Center}}
  }
  HorizontalBox_0 {widgetClassPath=&HorizontalBox, parent=VerticalBox_0, widget=HorizontalBox_0, slot {class=VerticalBoxSlot}}
  Image_0 {
   widgetClassPath=&Image
   bIsVariable=true
   parent=HorizontalBox_0
   widget=Image_0
   properties {brush {
     bIsDynamicallyLoaded=false
     drawAs=Image
     imageType=NoImage
     mirroring=NoMirror
     resourceName=None
     resourceObject=/Game/UI/Foundation/Textures/Icons/Social/T-Icon-EpicGamesLogo-64
     tiling=NoTile
     imageSize=<64 64>
     margin=<ltrb 0 0 0 0>
     outlineSettings {
      bUseBrushTransparency=false
      roundingType=HalfHeightRadius
      width=0
      color {colorUseRule=UseColor_Specified, specifiedColor=<rgba 0 0 0 0>}
      cornerRadii=<0 0 0 0>
     }
     tintColor {colorUseRule=UseColor_Specified, specifiedColor=<rgba 1 1 1 1>}
     uVRegion {bIsValid=false, max=<0 0>, min=<0 0>}
    }}
   slot {class=HorizontalBoxSlot, properties {horizontalAlignment=HAlign_Center, verticalAlignment=VAlign_Center}}
  }
  BorderItemNAme {
   widgetClassPath=&Border
   parent=HorizontalBox_0
   widget=BorderItemNAme
   properties {
    background {
     bIsDynamicallyLoaded=false
     drawAs=NoDrawType
     imageType=NoImage
     mirroring=NoMirror
     resourceName=None
     resourceObject=null
     tiling=NoTile
     imageSize=<32 32>
     margin=<ltrb 0 0 0 0>
     outlineSettings {
      bUseBrushTransparency=false
      roundingType=HalfHeightRadius
      width=0
      color {colorUseRule=UseColor_Specified, specifiedColor=<rgba 0 0 0 0>}
      cornerRadii=<0 0 0 0>
     }
     tintColor {colorUseRule=UseColor_Specified, specifiedColor=<rgba 1 1 1 1>}
     uVRegion {bIsValid=false, max=<0 0>, min=<0 0>}
    }
    padding=<ltrb 0 0 0 0>
   }
   slot {class=HorizontalBoxSlot}
  }
  Overlay_9 {widgetClassPath=&Overlay, parent=BorderItemNAme, widget=Overlay_9, slot {class=BorderSlot, properties {padding=<ltrb 0 0 0 0>}}}
  TextName {
   widgetClassPath=&CommonTextBlock
   bIsVariable=true
   parent=Overlay_9
   widget=TextName
   properties {
    justification=Center
    scrollStyle=&TextScrollStyle-Base_C
    style=/FortUILibrary/Styles/Typography/Foundation/TextStyle-Header-XL-Horizon.TextStyle-Header-XL-Horizon_C
    text=EPIC GAMES
    font {
     bForceMonospaced=false
     bMaterialIsStencil=false
     fontMaterial=/FortUILibrary/Materials/Foundation/M_Font_Horizon_Dynamic
     fontObject=&Burbank
     letterSpacing=0
     monospacedWidth=1
     size=72
     skewAmount=0
     typefaceFontName=Burbank-Black
     outlineSettings {
      bApplyOutlineToDropShadows=true
      bMiteredCorners=false
      bSeparateFillAlpha=false
      outlineMaterial=null
      outlineSize=3
      outlineColor=<rgba 0 0 0 0.1>
     }
    }
    shadowOffset=<0 0>
   }
   slot {class=OverlaySlot, properties {padding=<ltrb 0 -10 0 -20>}}
  }
  BorderSuccessText {
   widgetClassPath=&Border
   parent=VerticalBox_0
   widget=BorderSuccessText
   properties {
    visibility=Collapsed
    background {
     bIsDynamicallyLoaded=false
     drawAs=NoDrawType
     imageType=NoImage
     mirroring=NoMirror
     resourceName=None
     resourceObject=null
     tiling=NoTile
     imageSize=<32 32>
     margin=<ltrb 0 0 0 0>
     outlineSettings {
      bUseBrushTransparency=false
      roundingType=HalfHeightRadius
      width=0
      color {colorUseRule=UseColor_Specified, specifiedColor=<rgba 0 0 0 0>}
      cornerRadii=<0 0 0 0>
     }
     tintColor {colorUseRule=UseColor_Specified, specifiedColor=<rgba 1 1 1 1>}
     uVRegion {bIsValid=false, max=<0 0>, min=<0 0>}
    }
    padding=<ltrb 0 0 0 0>
   }
   slot {class=VerticalBoxSlot, properties {horizontalAlignment=HAlign_Center}}
  }
  Sender {
   widgetClassPath=&CommonTextBlock
   bIsVariable=true
   parent=BorderSuccessText
   widget=Sender
   properties {
    justification=Center
    style=/FortUILibrary/Styles/Typography/Foundation/TextStyle-Header-M.TextStyle-Header-M_C
    text="Sent you a gift box!"
    font {
     bForceMonospaced=false
     bMaterialIsStencil=false
     fontMaterial=null
     fontObject=&Burbank
     letterSpacing=0
     monospacedWidth=1
     size=32
     skewAmount=0
     typefaceFontName=Burbank-Black
     outlineSettings {
      bApplyOutlineToDropShadows=false
      bMiteredCorners=false
      bSeparateFillAlpha=false
      outlineMaterial=null
      outlineSize=0
      outlineColor=<rgba 0 0 0 1>
     }
    }
    shadowOffset=<0 0>
   }
   slot {class=BorderSlot, properties {padding=<ltrb 0 0 0 0>}}
  }
  VerticalBox_Text {widgetClassPath=&VerticalBox, bIsVariable=true, parent=VerticalBox_1, widget=VerticalBox_Text, slot {class=VerticalBoxSlot}}
  RichTextBlock_0 {
   widgetClassPath=/Script/UMG.RichTextBlock
   bIsVariable=true
   parent=VerticalBox_Text
   widget=RichTextBlock_0
   properties {
    justification=Center
    text="The following items have been restored to your account by player support.\r\nWe apologize for any inconvenience."
    textStyleSet=/Game/UI/ItemReceived/AthenaFrontendStyles
    wrapTextAt=1200
   }
   slot {class=VerticalBoxSlot, properties {horizontalAlignment=HAlign_Center, padding=<ltrb 0 20 0 0>}}
  }
 ]
}
